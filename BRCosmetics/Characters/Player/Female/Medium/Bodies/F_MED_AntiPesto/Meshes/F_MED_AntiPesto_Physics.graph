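# Physics asset for the F_MED_AntiPesto medium female player body cosmetic, covering only the bones of its jacket.
# Eight base jacket bones around front, back, left and right carry tiny capsules, each paired with a larger, offset capsule on a dynamic bone.
# Each dynamic bone is constrained to its base bone with zero twist and limited swing, narrow on one axis and wider on the other, widest at the back.
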PhysicsAsset /BRCosmetics/Characters/Player/Female/Medium/Bodies/F_MED_AntiPesto/Meshes/F_MED_AntiPesto_Physics
bodies {
 base_jacket_back [{shapeType=Capsule, length=0.505, radius=0.505, shapeName=base_jacket_back_capsule, rotation=<rot 0 -90 0>}]
 base_jacket_front [{shapeType=Capsule, length=0.505, radius=0.505, shapeName=base_jacket_front_capsule, rotation=<rot 0 -90 0>}]
 base_jacket_l_0 [{shapeType=Capsule, length=0.505, radius=0.505, shapeName=base_jacket_l_0_capsule, rotation=<rot 0 -90 0>}]
 base_jacket_l_1 [{shapeType=Capsule, length=0.505, radius=0.505, shapeName=base_jacket_l_1_capsule, rotation=<rot 0 -90 0>}]
 base_jacket_l_2 [{shapeType=Capsule, length=0.505, radius=0.505, shapeName=base_jacket_l_2_capsule, rotation=<rot 0 -90 0>}]
 base_jacket_r_0 [{shapeType=Capsule, length=0.505, radius=0.505, shapeName=base_jacket_r_0_capsule, rotation=<rot 0 -90 0>}]
 base_jacket_r_1 [{shapeType=Capsule, length=0.505, radius=0.505, shapeName=base_jacket_r_1_capsule, rotation=<rot 0 -90 0>}]
 base_jacket_r_2 [{shapeType=Capsule, length=0.505, radius=0.505, shapeName=base_jacket_r_2_capsule, rotation=<rot 0 -90 0>}]
 dyn_jacket_back [{shapeType=Capsule, length=2, radius=1.5, shapeName=dyn_jacket_back_capsule, center=<4 0 0>, rotation=<rot 90 0 0>}]
 dyn_jacket_front [{shapeType=Capsule, length=2, radius=1.5, shapeName=dyn_jacket_front_capsule, center=<4 0 0>, rotation=<rot 90 0 0>}]
 dyn_jacket_l_0 [{shapeType=Capsule, length=2, radius=1.5, shapeName=dyn_jacket_l_0_capsule, center=<4 0 0>, rotation=<rot 90 0 0>}]
 dyn_jacket_l_1 [{shapeType=Capsule, length=2, radius=1.5, shapeName=dyn_jacket_l_1_capsule, center=<4 0 0>, rotation=<rot 90 0 0>}]
 dyn_jacket_l_2 [{shapeType=Capsule, length=2, radius=1.5, shapeName=dyn_jacket_l_2_capsule, center=<4 0 0>, rotation=<rot 90 0 0>}]
 dyn_jacket_r_0 [{shapeType=Capsule, length=2, radius=1.5, shapeName=dyn_jacket_r_0_capsule, center=<4 0 0>, rotation=<rot 90 0 0>}]
 dyn_jacket_r_1 [{shapeType=Capsule, length=2, radius=1.5, shapeName=dyn_jacket_r_1_capsule, center=<4 0 0>, rotation=<rot 90 0 0>}]
 dyn_jacket_r_2 [{shapeType=Capsule, length=2, radius=1.5, shapeName=dyn_jacket_r_2_capsule, center=<4 0 0>, rotation=<rot 90 0 0>}]
}
constraints [
 {bone1Name=dyn_jacket_back, bone2Name=base_jacket_back, swing1LimitDegrees=10, swing2LimitDegrees=50, twistLimitDegrees=0}
 {bone1Name=dyn_jacket_front, bone2Name=base_jacket_front, swing1LimitDegrees=10, swing2LimitDegrees=30, twistLimitDegrees=0}
 {bone1Name=dyn_jacket_l_0, bone2Name=base_jacket_l_0, swing1LimitDegrees=10, swing2LimitDegrees=20, twistLimitDegrees=0}
 {bone1Name=dyn_jacket_l_1, bone2Name=base_jacket_l_1, swing1LimitDegrees=10, swing2LimitDegrees=20, twistLimitDegrees=0}
 {bone1Name=dyn_jacket_l_2, bone2Name=base_jacket_l_2, swing1LimitDegrees=10, swing2LimitDegrees=30, twistLimitDegrees=0}
 {bone1Name=dyn_jacket_r_0, bone2Name=base_jacket_r_0, swing1LimitDegrees=10, swing2LimitDegrees=20, twistLimitDegrees=0}
 {bone1Name=dyn_jacket_r_1, bone2Name=base_jacket_r_1, swing1LimitDegrees=10, swing2LimitDegrees=20, twistLimitDegrees=0}
 {bone1Name=dyn_jacket_r_2, bone2Name=base_jacket_r_2, swing1LimitDegrees=10, swing2LimitDegrees=30, twistLimitDegrees=0}
]
