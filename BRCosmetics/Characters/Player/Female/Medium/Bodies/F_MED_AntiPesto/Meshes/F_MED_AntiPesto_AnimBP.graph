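# Post-process style AnimBP for the F_MED_AntiPesto female body cosmetic, adding secondary motion to its bag, wrist straps and arm metal pieces.
# Copies the attached parent's pose through a Control Rig, then anchors accessory bones with weighted multi-bone constraints.
# Bag and upper-arm bone offsets precede RigidBody and spring-driven AnimDynamics, chained via cached poses.
# The event graph updates a gravity override (deprecated function) and right-thigh outward-rotation alpha.
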AnimBlueprint /BRCosmetics/Characters/Player/Female/Medium/Bodies/F_MED_AntiPesto/Meshes/F_MED_AntiPesto_AnimBP
target_skeleton=&/F_MED_AntiPesto_Skeleton
graphs [
 AnimGraph {nodes [
   Output Pose(Result=Use cached pose 'dynamicsCache'()) {layerGroup=None}
   Save cached pose 'iniPose'(Pose=Copy Pose From Mesh() {bUseAttachedParent=true}) {cacheName=iniPose}
   n1 = Control Rig(Source=Use cached pose 'iniPose'()) {
    alphaBoolBlend {blendOption=Linear}
    controlRigAssetReference {blueprintRigClass=&/F_MED_AntiPesto_CtrlRig.F_MED_AntiPesto_CtrlRig_C}
   }
   Save cached pose 'controlRigCache'(Pose=@n1) {cacheName=controlRigCache}
   n2 = RigidBody!(ComponentPose=Local To Component(LocalPose=Use cached pose 'modifyTransCache'()), Alpha, OverrideWorldGravity:Vector="0,0,0") -> (Pose) {
    bClampLinearTranslationLimitToRefPose=true
    bOverrideWorldGravity=true
    overridePhysicsAsset=&/F_MED_AntiPesto_Physics
    simulationSpace=BaseBoneSpace
    alphaBoolBlend {blendOption=Linear}
    baseBoneRef {boneName=spine_04}
    simSpaceSettings {dampingAlpha=0.8, maxAngularAcceleration=1200, maxLinearAcceleration=1200, worldAlpha=1}
   }
   n3 = Anim Dynamics - Bone: dyn_bag!(ComponentPose=@n2, Alpha=1) -> (Pose) {
    angularDampingOverride=1
    angularSpringConstant=10000
    bAngularSpring=true
    bOverrideAngularDamping=true
    bUseGravityOverride=true
    gravityScale=0.35
    simulationSpace=RootRelative
    alphaBoolBlend {blendOption=Linear}
    boundBone {boneName=dyn_bag}
    componentAppliedLinearAccClamp=<1000 1000 1000>
    componentLinearAccScale=<0.5 0.5 0.5>
    componentLinearVelScale=<0.5 0.5 0.5>
    gravityOverride {z=-980}
    physicsBodyDefinitions [{
      boundBone {boneName=dyn_bag}
      boxExtents=<30 30 30>
      constraintSetup {angularLimitsMax=<4 4 8>, angularLimitsMin=<-4 -4 0>, angularTarget=<1 0 0>}
      localJointOffset=<8 0 0>
     }]
   }
   n4 = Anim Dynamics - Bone: dyn_wrist_strap_r!(ComponentPose=@n3, Alpha=1) -> (Pose) {
    angularDampingOverride=1
    angularSpringConstant=10000
    bAngularSpring=true
    bOverrideAngularDamping=true
    bUseGravityOverride=true
    gravityScale=0.35
    simulationSpace=RootRelative
    alphaBoolBlend {blendOption=Linear}
    boundBone {boneName=dyn_wrist_strap_r}
    componentAppliedLinearAccClamp=<3000 3000 3000>
    componentLinearAccScale=<0.5 0.5 0.5>
    componentLinearVelScale=<0.5 0.5 0.5>
    gravityOverride {z=-980}
    physicsBodyDefinitions [{
      boundBone {boneName=dyn_wrist_strap_r}
      boxExtents=<30 30 30>
      constraintSetup {angularLimitsMax=<0 30 10>, angularLimitsMin=<0 -70 -10>, angularTarget=<1 0 0>}
      localJointOffset=<10 0 0>
     }]
   }
   n5 = Anim Dynamics - Bone: dyn_wrist_strap_l!(ComponentPose=@n4, Alpha=1) -> (Pose) {
    angularDampingOverride=1
    angularSpringConstant=10000
    bAngularSpring=true
    bOverrideAngularDamping=true
    bUseGravityOverride=true
    gravityScale=0.35
    simulationSpace=RootRelative
    alphaBoolBlend {blendOption=Linear}
    boundBone {boneName=dyn_wrist_strap_l}
    componentAppliedLinearAccClamp=<3000 3000 3000>
    componentLinearAccScale=<0.5 0.5 0.5>
    componentLinearVelScale=<0.5 0.5 0.5>
    gravityOverride {z=-980}
    physicsBodyDefinitions [{
      boundBone {boneName=dyn_wrist_strap_l}
      boxExtents=<30 30 30>
      constraintSetup {angularLimitsMax=<0 30 10>, angularLimitsMin=<0 -70 -10>, angularTarget=<1 0 0>}
      localJointOffset=<10 0 0>
     }]
   }
   n6 = Anim Dynamics - Bone: dyn_metal_piece_arm_l!(ComponentPose=@n5, Alpha=1) -> (Pose) {
    angularDampingOverride=1
    angularSpringConstant=8000
    bAngularSpring=true
    bGravityOverrideInSimSpace=true
    bOverrideAngularDamping=true
    bUseGravityOverride=true
    gravityScale=0.35
    simulationSpace=RootRelative
    alphaBoolBlend {blendOption=Linear}
    boundBone {boneName=dyn_metal_piece_arm_l}
    componentAppliedLinearAccClamp=<3000 3000 3000>
    componentLinearAccScale=<0.5 0.5 0.5>
    componentLinearVelScale=<0.5 0.5 0.5>
    gravityOverride {z=-980}
    physicsBodyDefinitions [{
      boundBone {boneName=dyn_metal_piece_arm_l}
      boxExtents=<20 20 20>
      constraintSetup {angularLimitsMax=<0 0 0>, angularLimitsMin=<0 -35 0>, angularTarget=<1 0 0>}
      localJointOffset=<6.5 0 0>
     }]
   }
   n7 = Anim Dynamics - Bone: dyn_metal_piece_arm_r!(ComponentPose=@n6, Alpha=1) -> (Pose) {
    angularDampingOverride=1
    angularSpringConstant=8000
    bAngularSpring=true
    bGravityOverrideInSimSpace=true
    bOverrideAngularDamping=true
    bUseGravityOverride=true
    gravityScale=0.35
    simulationSpace=RootRelative
    alphaBoolBlend {blendOption=Linear}
    boundBone {boneName=dyn_metal_piece_arm_r}
    componentAppliedLinearAccClamp=<3000 3000 3000>
    componentLinearAccScale=<0.5 0.5 0.5>
    componentLinearVelScale=<0.5 0.5 0.5>
    gravityOverride {z=-980}
    physicsBodyDefinitions [{
      boundBone {boneName=dyn_metal_piece_arm_r}
      boxExtents=<20 20 20>
      constraintSetup {angularLimitsMax=<0 0 0>, angularLimitsMin=<0 -35 0>, angularTarget=<1 0 0>}
      localJointOffset=<6.5 0 0>
     }]
   }
   Save cached pose 'dynamicsCache'(Pose=Component To Local(ComponentPose=@n7)) {cacheName=dynamicsCache}
   n8 = "Transform (Modify) Bone - Bone: deform_upperarm_up_l"!(ComponentPose=Local To Component(LocalPose=Use cached pose 'bagSetupCache'()), bAlphaBoolEnabled:bool=True, Translation="20,0,0") -> (Pose) {
    alphaInputType=Bool
    translationMode=BMM_Additive
    translationSpace=BCS_BoneSpace
    alphaBoolBlend {blendInTime=1, blendOption=Linear, blendOutTime=1}
    boneToModify {boneName=deform_upperarm_up_l}
    translation {x=20}
   }
   n9 = "Transform (Modify) Bone - Bone: deform_upperarm_up_r"!(ComponentPose=@n8, bAlphaBoolEnabled:bool=True, Translation="-20,0,0") -> (Pose) {
    alphaInputType=Bool
    translationMode=BMM_Additive
    translationSpace=BCS_BoneSpace
    alphaBoolBlend {blendInTime=1, blendOption=Linear, blendOutTime=1}
    boneToModify {boneName=deform_upperarm_up_r}
    translation {x=-20}
   }
   Save cached pose 'modifyTransCache'(Pose=Component To Local(ComponentPose=@n9)) {cacheName=modifyTransCache}
   n10 = "Constraint - base_wrist_strap_l (PPPPPP)"!(
    ComponentPose=Local To Component(LocalPose=Use cached pose 'controlRigCache'())
    Alpha=1
    ConstraintWeights_0=0.8496
    ConstraintWeights_1:float=0.0701
    ConstraintWeights_2:float=0.0458
    ConstraintWeights_3:float=0.027
    ConstraintWeights_4:float=0.0041
    ConstraintWeights_5:float=0.0035
   ) -> (Pose) {
    alphaBoolBlend {blendOption=Linear}
    boneToModify {boneName=base_wrist_strap_l}
    constraintSetup [
     {targetBone {boneName=lowerarm_twist_01_l}}
     {targetBone {boneName=lowerarm_l}}
     {targetBone {boneName=lowerarm_twist_02_l}}
     {targetBone {boneName=hand_l}}
     {targetBone {boneName=deform_wrist_up_l}}
     {targetBone {boneName=deform_wrist_dn_l}}
    ]
    constraintWeights [1,1,1,1,1,1]
   }
   n11 = "Constraint - base_wrist_strap_r (PPPPPP)"!(
    ComponentPose=@n10
    Alpha=1
    ConstraintWeights_0=0.8465
    ConstraintWeights_1:float=0.0735
    ConstraintWeights_2:float=0.0485
    ConstraintWeights_3:float=0.0251
    ConstraintWeights_4:float=0.0047
    ConstraintWeights_5:float=0.0018
   ) -> (Pose) {
    alphaBoolBlend {blendOption=Linear}
    boneToModify {boneName=base_wrist_strap_r}
    constraintSetup [
     {targetBone {boneName=lowerarm_twist_01_r}}
     {targetBone {boneName=lowerarm_r}}
     {targetBone {boneName=lowerarm_twist_02_r}}
     {targetBone {boneName=hand_r}}
     {targetBone {boneName=deform_wrist_up_r}}
     {targetBone {boneName=deform_wrist_dn_r}}
    ]
    constraintWeights [1,1,1,1,1,1]
   }
   n12 = "Constraint - base_bag (PPPPP)"!(
    ComponentPose=@n11
    Alpha=1
    ConstraintWeights_0=0.6891
    ConstraintWeights_1:float=0.2485
    ConstraintWeights_2:float=0.041
    ConstraintWeights_3:float=0.0201
    ConstraintWeights_4:float=0.0013
   ) -> (Pose) {
    alphaBoolBlend {blendOption=Linear}
    boneToModify {boneName=base_bag}
    constraintSetup [
     {targetBone {boneName=pelvis}}
     {targetBone {boneName=thigh_twist_01_r}}
     {targetBone {boneName=deform_groin_r}}
     {targetBone {boneName=deform_glute_r}}
     {targetBone {boneName=spine_02}}
    ]
    constraintWeights [1,1,1,1,1]
   }
   n13 = "Constraint - base_metal_piece_arm_l (PPPPPP)"!(
    ComponentPose=@n12
    Alpha=1
    ConstraintWeights_0=0.9457
    ConstraintWeights_1:float=0.0209
    ConstraintWeights_2:float=0.0182
    ConstraintWeights_3:float=0.0145
    ConstraintWeights_4:float=0.0005
    ConstraintWeights_5:float=0.0002
   ) -> (Pose) {
    alphaBoolBlend {blendOption=Linear}
    boneToModify {boneName=base_metal_piece_arm_l}
    constraintSetup [
     {targetBone {boneName=upperarm_twist_01_l}}
     {targetBone {boneName=upperarm_twist_02_l}}
     {targetBone {boneName=upperarm_l}}
     {targetBone {boneName=deform_upperarm_fr_l}}
     {targetBone {boneName=deform_upperarm_up_l}}
     {targetBone {boneName=clavicle_l}}
    ]
    constraintWeights [1,1,1,1,1,1]
   }
   n14 = "Constraint - base_metal_piece_arm_r (PPPPPP)"!(
    ComponentPose=@n13
    Alpha=1
    ConstraintWeights_0=0.9457
    ConstraintWeights_1:float=0.0195
    ConstraintWeights_2:float=0.017
    ConstraintWeights_3:float=0.0154
    ConstraintWeights_4:float=0.0017
    ConstraintWeights_5:float=0.0008
   ) -> (Pose) {
    alphaBoolBlend {blendOption=Linear}
    boneToModify {boneName=base_metal_piece_arm_r}
    constraintSetup [
     {targetBone {boneName=upperarm_twist_01_r}}
     {targetBone {boneName=upperarm_twist_02_r}}
     {targetBone {boneName=upperarm_r}}
     {targetBone {boneName=deform_upperarm_fr_r}}
     {targetBone {boneName=deform_upperarm_up_r}}
     {targetBone {boneName=clavicle_r}}
    ]
    constraintWeights [1,1,1,1,1,1]
   }
   Save cached pose 'surfaceConsCache'(Pose=Component To Local(ComponentPose=@n14)) {cacheName=surfaceConsCache}
   n15 = "Constraint - base_bag (P)"(ComponentPose=Local To Component(LocalPose=Use cached pose 'surfaceConsCache'()), Alpha=1, ConstraintWeights_0=1) {
    alphaBoolBlend {blendOption=Linear}
    boneToModify {boneName=base_bag}
    constraintSetup [{targetBone {boneName=thigh_twist_01_r}}]
    constraintWeights [1]
   }
   n16 = "Transform (Modify) Bone - Bone: dyn_bag_offset"!(ComponentPose=@n15, Alpha=1, Translation="8.714,-1.983,0", Rotation="0,13.2262,0") -> (Pose) {
    rotationMode=BMM_Additive
    rotationSpace=BCS_BoneSpace
    translationMode=BMM_Additive
    translationSpace=BCS_BoneSpace
    alphaBoolBlend {blendOption=Linear}
    alphaScaleBiasClamp {bInterpResult=true}
    boneToModify {boneName=dyn_bag_offset}
    rotation {yaw=13.2262}
    translation=<8.714 -1.983>
   }
   n17 = "Transform (Modify) Bone - Bone: base_bag_strap_offset"!(ComponentPose=@n16, Alpha=1, Translation="-1.9723,-3.5807,0.0002", Rotation="0,-20,0") -> (Pose) {
    rotationMode=BMM_Additive
    rotationSpace=BCS_BoneSpace
    translationMode=BMM_Additive
    translationSpace=BCS_BoneSpace
    alphaBoolBlend {blendOption=Linear}
    alphaScaleBiasClamp {bInterpResult=true}
    boneToModify {boneName=base_bag_strap_offset}
    rotation {yaw=-20}
    translation=<-1.9723 -3.5807 0.0002>
   }
   Save cached pose 'bagSetupCache'(Pose=Component To Local(ComponentPose=@n17)) {cacheName=bagSetupCache}
  ]}
 EventGraph {nodes [
   Event BlueprintUpdateAnimation()
   > n1 = Gravity Override State Function DEPRECATED(CharacterPartAnimInstance=Self-Reference(), GameTimeInSeconds=GetGameTimeInSeconds(), GravityOverrideParams=Get Gravity Override Params())
   > Set GravityOverride(GravityOverride=@n1)
   > n2 = Set Skel Mesh(Skel Mesh=GetOwnerSkeletalMeshComponent())
   > Set thigh_rotate_outward(thigh_rotate_outward:double=GetAlphaFromReferenceTransformDelta(SkeletalMeshComponent=@n2, BoneName=thigh_r, BaseName=pelvis, DeltaComponentType=EulerY, InputRange=(X=0,Y=60), TwistAxis="0,0,1")) -> (Output_Get:double)
  ]}
]
variables [
 Gravity Override Params {type=GravityOverrideParamsStruct}
 GravityOverride {type=Vector}
 Skel Mesh {type=SkeletalMeshComponent}
 thigh_rotate_outward {type=number}
]
