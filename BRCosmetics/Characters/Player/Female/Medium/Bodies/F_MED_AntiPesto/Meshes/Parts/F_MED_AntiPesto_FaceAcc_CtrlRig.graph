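# Control Rig for the AntiPesto female medium face-accessory part that drives corrective motion on hair bones from body pose.
# Spherical pose readers and a DeltaRotation function (head swing/twist remapped to a weight) feed weighted transform offsets.
# Aim Setup moves head_aim_tgt and aim-constrains three base hair bones to it.
# Front Hair and the main graph add arm, clavicle and spine correctives, some gated by isTargeting/IsCrouch, and write aim curves.
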ControlRigBlueprint /BRCosmetics/Characters/Player/Female/Medium/Bodies/F_MED_AntiPesto/Meshes/Parts/F_MED_AntiPesto_FaceAcc_CtrlRig
graphs [
 New Function_ContainedGraph {nodes [
   n1 = RigVMFunctionEntryNode()
   RigVMFunction_MathBoolToInteger = &RigVMFunction_MathBoolToInteger(Value=@n1.onlyTwist)
   RigVMFunction_MathQuaternionSwingTwist = &RigVMFunction_MathQuaternionSwingTwist(
    twistAxis=@n1.twistAxis
    Input=&RigUnit_GetRelativeTransformForItem(
     bChildInitial=False
     bParentInitial=False
     child=(Type=Bone,Name="None")
     parent=(Type=Bone,Name="None")
     child.Name=@n1.child
     parent.Name=@n1.parent
     parent.Type=@n1.parentType
    ).RelativeTransform.Rotation
   )
   RigVMFunction_MathQuaternionToEuler = &RigVMFunction_MathQuaternionToEuler(
    Value=RigVMDispatchNode(
     Values=((),())
     Index=@RigVMFunction_MathBoolToInteger.Result
     Values.0=@RigVMFunction_MathQuaternionSwingTwist.Swing
     Values.1=@RigVMFunction_MathQuaternionSwingTwist.Twist
    ) {template="DISPATCH_RigVMDispatch_SelectInt32(in Index,in Values,out Result)"}.Result
    RotationOrder=@n1.Rotation Order
   )
   > DISPATCH_RigVMDispatch_Print = RigVMDispatchNode(ScreenColor=(R=1,G=1,B=1,A=1), ScreenDuration=1.5, Enabled=@n1.debugRotation, Value=@RigVMFunction_MathQuaternionToEuler.Result) {template="DISPATCH_RigVMDispatch_Print(in Prefix,in Value,in Enabled,in ScreenDuration,in ScreenColor)"}
   Clamp = &RigVMFunction_MathIntClamp(Value=@n1.primAxis)
   DISPATCH_RigVMDispatch_SelectInt32 = RigVMDispatchNode(
    Values=(0,0,0)
    Index=@Clamp.Result
    Values.0=@RigVMFunction_MathQuaternionToEuler.Result.X
    Values.1=@RigVMFunction_MathQuaternionToEuler.Result.Y
    Values.2=@RigVMFunction_MathQuaternionToEuler.Result.Z
   ) {template="DISPATCH_RigVMDispatch_SelectInt32(in Index,in Values,out Result)"}
   Remap = &RigVMFunction_MathDoubleRemap(
    TargetMaximum=1
    TargetMinimum=0
    bClamp=False
    Value=@DISPATCH_RigVMDispatch_SelectInt32.Result
    SourceMaximum=@n1.max
    SourceMinimum=@n1.min
   )
   Multiply = &RigVMFunction_MathDoubleMul(A=&RigVMFunction_MathDoubleClamp(Value=@Remap.Result).Result, B=@n1.premult)
   > RigVMFunctionReturnNode(outputWeight=@Multiply.Result)
  ]}
 New Function_ContainedGraph {nodes [
   n1 = RigVMFunctionEntryNode()
   RigVMFunction_Sequence = &RigVMFunction_Sequence(ExecuteContext=@n1.ExecuteContext)
   RigVMFunction_Sequence_1 = RigVMAggregateNode(ExecuteContext=@RigVMFunction_Sequence.A)
   DeltaRotation = RigVMFunctionReferenceNode(
    Rotation Order=ZYX
    child=head
    debugRotation=false
    max=-40
    parent=head_null
    parentType=Null
    premult=1
    primAxis=2
    twistAxis=(X=1,Y=0,Z=0)
    ExecuteContext=@RigVMFunction_Sequence_1.A
   )
   > &RigUnit_ModifyTransforms(
    ItemToModify="((Item=(Type=Bone,Name=\"head_aim_tgt\"),Transform=(Rotation=(X=0,Y=0,Z=0,W=1),Translation=(X=0,Y=-2,Z=-10),Scale3D=(X=1,Y=1,Z=1))),(Item=(Type=Bone,Name=\"base_hair_rt\"),Transform=(Rotation=(X=0,Y=0,Z=0.1736,W=0.9848),Translation=(X=0,Y=0,Z=0),Scale3D=(X=1,Y ...(truncated)"
    WeightMinimum=0
    Weight=@DeltaRotation.outputWeight
   )
   > &RigUnit_ModifyTransforms(
    ItemToModify="((Item=(Type=Bone,Name=\"base_hair_front_side_01_01_lf\"),Transform=(Rotation=(X=0,Y=0.4226,Z=0,W=0.9063),Translation=(X=-1.0208,Y=0,Z=0.0544),Scale3D=(X=1,Y=1,Z=1))),(Item=(Type=Bone,Name=\"dyn_hair_front_side_02_01_lf\"),Transform=(Rotation=(X=0,Y=0.2588,Z=0 ...(truncated)"
    WeightMinimum=0
    Weight=@DeltaRotation.outputWeight
   )
   > DeltaRotation_1 = RigVMFunctionReferenceNode(
    Rotation Order=ZYX
    child=head
    debugRotation=false
    max=40
    parent=head_null
    parentType=Null
    premult=1
    primAxis=2
    twistAxis=(X=1,Y=0,Z=0)
   )
   > &RigUnit_ModifyTransforms(
    ItemToModify=((Item=(Type=Bone,Name="head_aim_tgt"),Transform=(Rotation=(X=0,Y=0,Z=0,W=1),Translation=(X=0,Y=-2,Z=0),Scale3D=(X=1,Y=1,Z=1))))
    WeightMinimum=0
    Weight=@DeltaRotation_1.outputWeight
   )
   n2 = &RigUnit_SphericalPoseReader(
    ActiveRegionScaleFactors=(PositiveWidth=0.75,NegativeWidth=0.75,PositiveHeight=1,NegativeHeight=1)
    ActiveRegionSize=0.1
    Debug=(bDrawDebug=true,bDraw2D=false,DebugScale=25,DebugSegments=20,DebugThickness=0.25)
    DriverAxis=(X=0,Y=1,Z=0)
    DriverItem=(Type=Bone,Name="head")
    FalloffRegionScaleFactors=(PositiveWidth=1,NegativeWidth=1,PositiveHeight=1,NegativeHeight=2.2)
    FalloffSize=0.2
    OptionalParentItem=(Type=Bone)
    RotationOffset=(X=150,Y=0,Z=0)
    ExecutePin=@RigVMFunction_Sequence_1.B
   )
   > &RigUnit_ModifyTransforms(
    ItemToModify="((Item=(Type=Bone,Name=\"head_aim_tgt\"),Transform=(Rotation=(X=0,Y=0,Z=0,W=1),Translation=(X=-5,Y=-1,Z=0),Scale3D=(X=1,Y=1,Z=1))),(Item=(Type=Bone,Name=\"base_hair_rt\"),Transform=(Rotation=(X=0,Y=0,Z=0,W=1),Translation=(X=0,Y=0,Z=5),Scale3D=(X=1,Y=1,Z=1))),( ...(truncated)"
    WeightMinimum=0
    Weight=@n2.OutputParam
   )
   > &RigUnit_ModifyTransforms(
    ItemToModify=((Item=(Type=Bone,Name="base_hair_lf"),Transform=(Rotation=(X=0,Y=0,Z=0,W=1),Translation=(X=0,Y=4,Z=0),Scale3D=(X=1,Y=1,Z=1))))
    WeightMinimum=0
    Weight=&RigVMFunction_MathDoubleRemap(SourceMaximum=0.3, SourceMinimum=0, TargetMaximum=1, TargetMinimum=0, Value=@n2.OutputParam).Result
   )
   > n3 = &RigUnit_SphericalPoseReader(
    ActiveRegionScaleFactors=(PositiveWidth=0.75,NegativeWidth=0.75,PositiveHeight=1.5,NegativeHeight=1)
    ActiveRegionSize=0.1
    Debug=(bDrawDebug=false,bDraw2D=false,DebugScale=25,DebugSegments=20,DebugThickness=0.25)
    DriverAxis=(X=0,Y=1,Z=0)
    DriverItem=(Type=Bone,Name="head")
    FalloffRegionScaleFactors=(PositiveWidth=1,NegativeWidth=1,PositiveHeight=1,NegativeHeight=2.2)
    FalloffSize=0.2
    FlipHeightScaling=true
    FlipWidthScaling=false
    OptionalParentItem=(Type=Bone)
    RotationOffset=(X=30,Y=0,Z=0)
   )
   > &RigUnit_ModifyTransforms(
    ItemToModify="((Item=(Type=Bone,Name=\"head_aim_tgt\"),Transform=(Rotation=(X=0,Y=0,Z=0,W=-1),Translation=(X=1.726,Y=-1,Z=0),Scale3D=(X=1,Y=1,Z=1))),(Item=(Type=Bone,Name=\"base_hair_lf\"),Transform=(Rotation=(X=0,Y=0,Z=0.2164,W=-0.9763),Translation=(X=0,Y=0,Z=1.5),Scale3D= ...(truncated)"
    WeightMinimum=0
    Weight=@n3.OutputParam
   )
   &RigUnit_AimConstraintLocalSpaceOffset(
    Child=(Type=Bone,Name="base_hair_lf")
    Filter=(bX=True,bY=true,bZ=true)
    Parents=((Item=(Type=Bone,Name="head_aim_tgt"),Weight=1))
    UpAxis=(X=0,Y=-1,Z=0)
    Weight=1
    WorldUp=(Target=(X=0,Y=-1,Z=0),Kind=Location,Space=(Type=Bone,Name="head_aim_tgt"))
    bMaintainOffset=true
    ExecutePin=@RigVMFunction_Sequence_1.C
   )
   > &RigUnit_AimConstraintLocalSpaceOffset(
    Child=(Type=Bone,Name="base_hair_cr")
    Filter=(bX=True,bY=true,bZ=true)
    Parents=((Item=(Type=Bone,Name="head_aim_tgt"),Weight=1))
    UpAxis=(X=0,Y=-1,Z=0)
    Weight=1
    WorldUp=(Target=(X=0,Y=-1,Z=0),Kind=Location,Space=(Type=Bone,Name="head_aim_tgt"))
    bMaintainOffset=true
   )
   > &RigUnit_AimConstraintLocalSpaceOffset(
    Child=(Type=Bone,Name="base_hair_rt")
    Filter=(bX=True,bY=true,bZ=true)
    Parents=((Item=(Type=Bone,Name="head_aim_tgt"),Weight=1))
    UpAxis=(X=0,Y=-1,Z=0)
    Weight=1
    WorldUp=(Target=(X=0,Y=-1,Z=0),Kind=Location,Space=(Type=Bone,Name="head_aim_tgt"))
    bMaintainOffset=true
   )
   > &RigUnit_RotationConstraintLocalSpaceOffset(
    AdvancedSettings=(InterpolationType=Average,RotationOrderForFilter=XZY)
    Child=(Type=Bone,Name="base_hair_front_side_02_01_lf")
    Filter=(bX=True,bY=True,bZ=True)
    Parents=((Item=(Type=Bone,Name="head"),Weight=0.6),(Item=(Type=Bone,Name="neck_01"),Weight=0.4))
    Weight=1
   )
   > &RigUnit_ModifyTransforms(
    ItemToModify=((Item=(Type=Bone,Name="base_hair_lf"),Transform=(Rotation=(X=0.0019,Y=0.0436,Z=-0.0436,W=0.9981),Translation=(X=0,Y=0,Z=0),Scale3D=(X=1,Y=1,Z=1))))
    WeightMinimum=0
    Weight=@n3.OutputParam
   )
   > &RigUnit_ModifyTransforms(
    ItemToModify="((Item=(Type=Bone,Name=\"base_hair_lf\"),Transform=(Rotation=(X=0,Y=0,Z=0,W=1),Translation=(X=0,Y=3.2576,Z=0),Scale3D=(X=1,Y=1,Z=1))),(Item=(Type=Bone,Name=\"base_hair_front_side_02_01_lf\"),Transform=(Rotation=(X=0,Y=-0.2588,Z=0,W=0.9659),Translation=(X=0,Y=0 ...(truncated)"
    WeightMinimum=0
    Weight=@DeltaRotation.outputWeight
   )
   > &RigUnit_ModifyTransforms(
    ItemToModify="((Item=(Type=Bone,Name=\"base_hair_rt\"),Transform=(Rotation=(X=0.0075,Y=-0.0013,Z=-0.1312,W=-0.9913),Translation=(X=0,Y=0,Z=0),Scale3D=(X=1,Y=1,Z=1))),(Item=(Type=Bone,Name=\"base_hair_front_side_02_01_lf\"),Transform=(Rotation=(X=-0.0114,Y=0.0864,Z=0.13,W=0. ...(truncated)"
    WeightMinimum=0
    Weight=@n2.OutputParam
   )
   > &RigUnit_ModifyTransforms(
    ItemToModify=((Item=(Type=Bone,Name="dyn_hair_front_side_01_01_lf"),Transform=(Rotation=(X=0,Y=0,Z=0,W=1),Translation=(X=0,Y=1,Z=0),Scale3D=(X=1,Y=1,Z=1))))
    WeightMinimum=0
    Weight=@n2.OutputParam
   )
   > &RigUnit_ModifyTransforms(
    ItemToModify=((Item=(Type=Bone,Name="dyn_hair_front_side_01_01_lf"),Transform=(Rotation=(X=0,Y=0,Z=0,W=1),Translation=(X=0,Y=0.4,Z=0),Scale3D=(X=1,Y=1,Z=1))))
    Weight=1
    WeightMinimum=0
   )
   RigVMFunctionReturnNode(ExecuteContext=@RigVMFunction_Sequence.B)
  ]}
 New Function_ContainedGraph {nodes [
   n1 = RigVMFunctionEntryNode()
   RigVMFunction_Sequence = &RigVMFunction_Sequence(ExecuteContext=@n1.ExecuteContext)
   RigVMFunction_Sequence_1 = RigVMAggregateNode(ExecuteContext=@RigVMFunction_Sequence.A)
   n2 = &RigUnit_SphericalPoseReader(
    ActiveRegionScaleFactors=(PositiveWidth=1,NegativeWidth=1,PositiveHeight=1,NegativeHeight=1)
    ActiveRegionSize=0.1
    Debug=(bDrawDebug=false,bDraw2D=False,bDrawLocalAxes=False,DebugScale=25,DebugSegments=20,DebugThickness=0.25)
    DriverAxis=(X=-1,Y=0,Z=0)
    DriverItem=(Type=Bone,Name="upperarm_r")
    FalloffRegionScaleFactors=(PositiveWidth=3.5,NegativeWidth=3.5,PositiveHeight=3.5,NegativeHeight=3.5)
    FalloffSize=0.15
    FlipHeightScaling=False
    FlipWidthScaling=False
    OptionalParentItem=(Type=Bone,Name="None")
    RotationOffset=(X=-100,Y=0,Z=20)
    ExecutePin=@RigVMFunction_Sequence_1.A
   )
   RigVMFunction_MathBoolNot = &RigVMFunction_MathBoolNot(Value=RigVMVariableNode() {variable=IsCrouch}.Value)
   RigVMFunction_MathBoolAnd = &RigVMFunction_MathBoolAnd(B=@RigVMFunction_MathBoolNot.Result, A=RigVMVariableNode() {variable=isTargeting}.Value)
   RigVMFunction_MathBoolToFloat = &RigVMFunction_MathBoolToFloat(Value=@RigVMFunction_MathBoolAnd.Result)
   Multiply = &RigVMFunction_MathDoubleMul(B=@RigVMFunction_MathBoolToFloat.Result, A=@n2.OutputParam)
   AlphaInterp = &RigVMFunction_AlphaInterp(InterpSpeedDecreasing=10, InterpSpeedIncreasing=10, Scale=1, bClampResult=false, bMapRange=false, Value=@Multiply.Result)
   > &RigUnit_ModifyTransforms(
    ItemToModify="((Item=(Type=Bone,Name=\"dyn_hair_front_side_02_02_lf\"),Transform=(Rotation=(X=0,Y=0,Z=-0.1736,W=-0.9848),Translation=(X=-1.2091,Y=-0.4401,Z=0),Scale3D=(X=1,Y=1,Z=1))),(Item=(Type=Bone,Name=\"dyn_hair_front_side_01_02_lf\"),Transform=(Rotation=(X=0,Y=0,Z=-0.1 ...(truncated)"
    WeightMinimum=0
    Weight=@AlphaInterp.Result
   )
   > DeltaRotation = RigVMFunctionReferenceNode(
    Rotation Order=ZYX
    child=head
    debugRotation=false
    max=-40
    parent=head_null
    parentType=Null
    premult=1
    primAxis=2
    twistAxis=(X=1,Y=0,Z=0)
   )
   n3 = &RigVMFunction_MathDoubleMul(B=@AlphaInterp.Result, A=@DeltaRotation.outputWeight)
   > &RigUnit_ModifyTransforms(
    ItemToModify="((Item=(Type=Bone,Name=\"dyn_hair_front_side_02_01_lf\"),Transform=(Rotation=(X=-0.0226,Y=0.0842,Z=0.2578,W=0.9623),Translation=(X=0,Y=0,Z=0),Scale3D=(X=1,Y=1,Z=1))),(Item=(Type=Bone,Name=\"dyn_hair_front_side_01_01_lf\"),Transform=(Rotation=(X=-0.0189,Y=0.085 ...(truncated)"
    WeightMinimum=0
    Weight=@n3.Result
   )
   > /Script/ControlRig.RigUnit_SetCurveValue(
    Curve=Aim_Up
    Value=RigVMDispatchNode(False=0, True=@n3.Result, Condition=RigVMVariableNode() {variable=isTargeting}.Value) {template="DISPATCH_RigVMDispatch_If(in Condition,in True,in False,out Result)"}.Result
   ) {is=struct RigVMUnitNode}
   > n4 = &RigUnit_SphericalPoseReader(
    ActiveRegionScaleFactors=(PositiveWidth=1,NegativeWidth=1,PositiveHeight=1,NegativeHeight=1)
    ActiveRegionSize=0.1
    Debug=(bDrawDebug=false,bDraw2D=False,bDrawLocalAxes=False,DebugScale=25,DebugSegments=20,DebugThickness=0.25)
    DriverAxis=(X=0,Y=1,Z=0)
    DriverItem=(Type=Bone,Name="pelvis")
    FalloffRegionScaleFactors=(PositiveWidth=1,NegativeWidth=1,PositiveHeight=1,NegativeHeight=1)
    FalloffSize=0.2
    FlipHeightScaling=False
    FlipWidthScaling=False
    OptionalParentItem=(Type=Bone,Name="spine_05")
    RotationOffset=(X=150,Y=0,Z=0)
   )
   > &RigUnit_ModifyTransforms(
    ItemToModify="((Item=(Type=Bone,Name=\"dyn_hair_front_side_02_02_lf\"),Transform=(Rotation=(X=0,Y=0,Z=0,W=1),Translation=(X=0,Y=0,Z=0),Scale3D=(X=1,Y=1,Z=1))),(Item=(Type=Bone,Name=\"dyn_hair_front_side_01_01_lf\"),Transform=(Rotation=(X=0,Y=0,Z=0,W=1),Translation=(X=0,Y=0, ...(truncated)"
    WeightMinimum=0
    Weight=RigVMDispatchNode(False=0, True=@n4.OutputParam, Condition=RigVMVariableNode() {variable=isTargeting}.Value) {template="DISPATCH_RigVMDispatch_If(in Condition,in True,in False,out Result)"}.Result
   )
   n5 = &RigUnit_SphericalPoseReader(
    ActiveRegionScaleFactors=(PositiveWidth=1,NegativeWidth=1,PositiveHeight=1,NegativeHeight=1)
    ActiveRegionSize=0.12
    Debug=(bDrawDebug=false,bDraw2D=False,bDrawLocalAxes=False,DebugScale=25,DebugSegments=20,DebugThickness=0.25)
    DriverAxis=(X=1,Y=0,Z=0)
    DriverItem=(Type=Bone,Name="clavicle_l")
    FalloffRegionScaleFactors=(PositiveWidth=1,NegativeWidth=1,PositiveHeight=1,NegativeHeight=1)
    FalloffSize=0.25
    FlipHeightScaling=False
    FlipWidthScaling=false
    OptionalParentItem=(Type=Bone,Name="None")
    RotationOffset=(X=0,Y=-45,Z=0)
    ExecutePin=@RigVMFunction_Sequence_1.B
   )
   > &RigUnit_ModifyTransforms(
    ItemToModify=((Item=(Type=Bone,Name="base_hair_front_side_02_01_rt"),Transform=(Rotation=(X=0,Y=0.1736,Z=0,W=-0.9848),Translation=(X=-1.888,Y=0,Z=0),Scale3D=(X=1,Y=1,Z=1))))
    WeightMinimum=0
    Weight=@n5.OutputParam
   )
   > n6 = &RigUnit_SphericalPoseReader(
    ActiveRegionScaleFactors=(PositiveWidth=1,NegativeWidth=1,PositiveHeight=1,NegativeHeight=1)
    ActiveRegionSize=0.12
    Debug=(bDrawDebug=false,bDraw2D=False,bDrawLocalAxes=False,DebugScale=25,DebugSegments=20,DebugThickness=0.25)
    DriverAxis=(X=-1,Y=0,Z=0)
    DriverItem=(Type=Bone,Name="clavicle_r")
    FalloffRegionScaleFactors=(PositiveWidth=1,NegativeWidth=1,PositiveHeight=1,NegativeHeight=1)
    FalloffSize=0.25
    FlipHeightScaling=false
    FlipWidthScaling=false
    OptionalParentItem=(Type=Bone,Name="None")
    RotationOffset=(X=0,Y=135,Z=0)
   )
   > &RigUnit_ModifyTransforms(
    ItemToModify="((Item=(Type=Bone,Name=\"base_hair_front_side_02_01_lf\"),Transform=(Rotation=(X=0,Y=-0.2588,Z=0,W=0.9659),Translation=(X=-0.4371,Y=0,Z=-0.3668),Scale3D=(X=1,Y=1,Z=1))),(Item=(Type=Bone,Name=\"base_hair_front_side_01_01_lf\"),Transform=(Rotation=(X=0,Y=-0.1736 ...(truncated)"
    WeightMinimum=0
    Weight=@n6.OutputParam
   )
   > n7 = &RigUnit_SphericalPoseReader(
    ActiveRegionScaleFactors=(PositiveWidth=1,NegativeWidth=1,PositiveHeight=1,NegativeHeight=1)
    ActiveRegionSize=0.01
    Debug=(bDrawDebug=false,bDraw2D=False,bDrawLocalAxes=False,DebugScale=25,DebugSegments=20,DebugThickness=0.25)
    DriverAxis=(X=0,Y=1,Z=0)
    DriverItem=(Type=Bone,Name="head")
    FalloffRegionScaleFactors=(PositiveWidth=1,NegativeWidth=1,PositiveHeight=1,NegativeHeight=1)
    FalloffSize=0.2
    FlipHeightScaling=False
    FlipWidthScaling=False
    OptionalParentItem=(Type=Bone,Name="clavicle_r")
    RotationOffset=(X=0,Y=-87,Z=-5)
   )
   > &RigUnit_ModifyTransforms(
    ItemToModify="((Item=(Type=Bone,Name=\"dyn_hair_front_side_02_01_lf\"),Transform=(Rotation=(X=0,Y=-0.0872,Z=0,W=0.9962),Translation=(X=0,Y=0,Z=0),Scale3D=(X=1,Y=1,Z=1))),(Item=(Type=Bone,Name=\"dyn_hair_front_side_02_02_lf\"),Transform=(Rotation=(X=-0.0151,Y=0.0858,Z=0.173, ...(truncated)"
    WeightMinimum=0
    Weight=RigVMDispatchNode(False=0, True=@n7.OutputParam, Condition=RigVMVariableNode() {variable=IsCrouch}.Value) {template="DISPATCH_RigVMDispatch_If(in Condition,in True,in False,out Result)"}.Result
   )
   > n8 = &RigUnit_SphericalPoseReader(
    ActiveRegionScaleFactors=(PositiveWidth=1,NegativeWidth=5,PositiveHeight=1,NegativeHeight=5)
    ActiveRegionSize=0.05
    Debug=(bDrawDebug=false,bDraw2D=False,bDrawLocalAxes=False,DebugScale=25,DebugSegments=20,DebugThickness=0.25)
    DriverAxis=(X=0,Y=1,Z=0)
    DriverItem=(Type=Bone,Name="head")
    FalloffRegionScaleFactors=(PositiveWidth=1,NegativeWidth=1,PositiveHeight=1,NegativeHeight=1)
    FalloffSize=0.27
    FlipHeightScaling=False
    FlipWidthScaling=False
    OptionalParentItem=(Type=Bone,Name="clavicle_r")
    RotationOffset=(X=0,Y=-65,Z=20)
   )
   n9 = RigVMDispatchNode(
    False=0
    Condition=&RigVMFunction_MathBoolAnd(A=RigVMVariableNode() {variable=IsCrouch}.Value, B=RigVMVariableNode() {variable=isTargeting}.Value).Result
    True=@n8.OutputParam
   ) {template="DISPATCH_RigVMDispatch_If(in Condition,in True,in False,out Result)"}
   > &RigUnit_ModifyTransforms(
    ItemToModify="((Item=(Type=Bone,Name=\"dyn_hair_front_side_02_01_lf\"),Transform=(Rotation=(X=-0.0076,Y=0.0868,Z=0.0868,W=0.9924),Translation=(X=0,Y=0,Z=0),Scale3D=(X=1,Y=1,Z=1))),(Item=(Type=Bone,Name=\"dyn_hair_front_side_02_02_lf\"),Transform=(Rotation=(X=0,Y=0,Z=0.1736, ...(truncated)"
    WeightMinimum=0
    Weight=@n9.Result
   )
   > &RigUnit_SetCurveValue(Curve=Aim_Pistol_Crouch, Value=@n9.Result)
   Remap = &RigVMFunction_MathDoubleRemap(SourceMaximum=0.2, SourceMinimum=0, TargetMaximum=1, TargetMinimum=0, Value=@n9.Result)
   > &RigUnit_ModifyTransforms(
    ItemToModify="((Item=(Type=Bone,Name=\"dyn_hair_front_side_01_02_lf\"),Transform=(Rotation=(X=0,Y=0,Z=0.2588,W=0.9659),Translation=(X=0,Y=0,Z=0),Scale3D=(X=1,Y=1,Z=1))),(Item=(Type=Bone,Name=\"dyn_hair_front_side_01_02_lf\"),Transform=(Rotation=(X=0,Y=0,Z=0.0872,W=0.9962),T ...(truncated)"
    WeightMinimum=0
    Weight=@Remap.Result
   )
   > n10 = &RigUnit_SphericalPoseReader(
    ActiveRegionScaleFactors=(PositiveWidth=1,NegativeWidth=1,PositiveHeight=1,NegativeHeight=1)
    ActiveRegionSize=0.05
    Debug=(bDrawDebug=true,bDraw2D=False,bDrawLocalAxes=False,DebugScale=25,DebugSegments=20,DebugThickness=0.25)
    DriverAxis=(X=-1,Y=0,Z=0)
    DriverItem=(Type=Bone,Name="upperarm_r")
    FalloffRegionScaleFactors=(PositiveWidth=1,NegativeWidth=1,PositiveHeight=1,NegativeHeight=1)
    FalloffSize=0.1
    FlipHeightScaling=False
    FlipWidthScaling=False
    OptionalParentItem=(Type=Bone,Name="None")
    RotationOffset=(X=-180,Y=70,Z=-30)
   )
   > n11 = &RigUnit_SphericalPoseReader(
    ActiveRegionScaleFactors=(PositiveWidth=1,NegativeWidth=1,PositiveHeight=1,NegativeHeight=1)
    ActiveRegionSize=0.05
    Debug=(bDrawDebug=true,bDraw2D=False,bDrawLocalAxes=False,DebugScale=25,DebugSegments=20,DebugThickness=0.25)
    DriverAxis=(X=1,Y=1,Z=0)
    DriverItem=(Type=Bone,Name="pelvis")
    FalloffRegionScaleFactors=(PositiveWidth=1,NegativeWidth=1,PositiveHeight=1,NegativeHeight=1)
    FalloffSize=0.7
    FlipHeightScaling=False
    FlipWidthScaling=False
    OptionalParentItem=(Type=Bone,Name="spine_05")
    RotationOffset=(X=0,Y=-90,Z=0)
   )
   > &RigUnit_ModifyTransforms(
    ItemToModify="((Item=(Type=Bone,Name=\"base_hair_front_side_01_01_lf\"),Transform=(Rotation=(X=-0.0302,Y=0.171,Z=0.171,W=0.9698),Translation=(X=0,Y=0,Z=0),Scale3D=(X=1,Y=1,Z=1))),(Item=(Type=Bone,Name=\"dyn_hair_front_side_01_02_lf\"),Transform=(Rotation=(X=-0.1736,Y=0,Z=0, ...(truncated)"
    WeightMinimum=0
    Weight=&RigVMFunction_MathDoubleMul(B=@n10.OutputParam, A=@n11.OutputParam).Result
   )
   RigVMFunctionReturnNode(ExecuteContext=@RigVMFunction_Sequence.B)
  ]}
 RigVMFunctionLibrary {nodes [
   Aim Setup = RigVMCollapseNode(ExecuteContext=())
   DeltaRotation = RigVMCollapseNode(ExecuteContext=(), Rotation Order=ZYX, premult=1, twistAxis=(X=1,Y=0,Z=0))
   Front Hair = RigVMCollapseNode(ExecuteContext=())
  ]}
 RigVMFunction_Sequence_1_ContainedGraph {nodes [
   n1 = RigVMFunctionEntryNode()
   RigVMFunction_Sequence = &RigVMFunction_Sequence(ExecuteContext=@n1.ExecuteContext)
   n2 = &RigVMFunction_Sequence(ExecuteContext=@RigVMFunction_Sequence.B)
   RigVMFunctionReturnNode(A=@RigVMFunction_Sequence.A, B=@n2.A, C=@n2.B)
  ]}
 RigVMFunction_Sequence_2_ContainedGraph {nodes [
   n1 = RigVMFunctionEntryNode()
   n2 = &RigVMFunction_Sequence(ExecuteContext=@n1.ExecuteContext)
   n3 = &RigVMFunction_Sequence(ExecuteContext=@n2.B)
   n4 = &RigVMFunction_Sequence(ExecuteContext=@n3.B)
   RigVMFunctionReturnNode(A=@n2.A, B=@n3.A, C=@n4.A, D=@n4.B)
  ]}
 RigVMFunction_Sequence_2_ContainedGraph {nodes [
   n1 = RigVMFunctionEntryNode()
   n2 = &RigVMFunction_Sequence(ExecuteContext=@n1.ExecuteContext)
   n3 = &RigVMFunction_Sequence(ExecuteContext=@n2.B)
   RigVMFunctionReturnNode(A=@n2.A, B=@n3.A, C=@n3.B)
  ]}
 RigVMModel {nodes [
   RigUnit_BeginExecution = &RigUnit_BeginExecution()
   RigVMFunction_Sequence = RigVMAggregateNode(ExecuteContext=@RigUnit_BeginExecution.ExecutePin)
   Aim Setup = RigVMFunctionReferenceNode(ExecuteContext=@RigVMFunction_Sequence.A)
   > Front Hair = RigVMFunctionReferenceNode()
   n1 = &RigUnit_SphericalPoseReader(
    ActiveRegionScaleFactors=(PositiveWidth=1,NegativeWidth=1,PositiveHeight=1,NegativeHeight=1)
    ActiveRegionSize=0.07
    Debug=(bDrawDebug=true,bDraw2D=False,bDrawLocalAxes=False,DebugScale=25,DebugSegments=20,DebugThickness=0.25)
    DriverAxis=(X=1,Y=0,Z=0)
    DriverItem=(Type=Bone,Name="upperarm_l")
    FalloffRegionScaleFactors=(PositiveWidth=1,NegativeWidth=1,PositiveHeight=1,NegativeHeight=1)
    FalloffSize=0.12
    FlipHeightScaling=False
    FlipWidthScaling=False
    OptionalParentItem=(Type=Bone,Name="None")
    RotationOffset=(X=50,Y=-30,Z=-80)
    ExecutePin=@RigVMFunction_Sequence.B
   )
   AlphaInterp = &RigVMFunction_AlphaInterp(
    InterpSpeedDecreasing=8
    InterpSpeedIncreasing=8
    Scale=1
    bClampResult=true
    bMapRange=true
    Value=RigVMDispatchNode(False=0, True=@n1.OutputParam, Condition=RigVMVariableNode() {variable=isTargeting}.Value) {template="DISPATCH_RigVMDispatch_If(in Condition,in True,in False,out Result)"}.Result
   )
   RigVMFunction_AnimEvalRichCurve = &RigVMFunction_AnimEvalRichCurve(
    Curve=(EditorCurveData=(Keys=((InterpMode=RCIM_Cubic),(InterpMode=RCIM_Cubic,Time=1,Value=1))))
    SourceMinimum=0
    TargetMinimum=0
    Value=@AlphaInterp.Result
   )
   > &RigUnit_SetCurveValue(Curve=Aim_Up_Fusile, Value=@RigVMFunction_AnimEvalRichCurve.Result)
   > &RigUnit_ModifyTransforms(
    ItemToModify="((Item=(Type=Bone,Name=\"base_hair_rt\"),Transform=(Rotation=(X=-0.0151,Y=-0.173,Z=-0.0858,W=0.9811),Translation=(X=0,Y=0,Z=0),Scale3D=(X=1,Y=1,Z=1))),(Item=(Type=Bone,Name=\"base_hair_lf\"),Transform=(Rotation=(X=0,Y=0,Z=0.0872,W=0.9962),Translation=(X=0,Y=0, ...(truncated)"
    WeightMinimum=0
    Weight=@RigVMFunction_AnimEvalRichCurve.Result
   )
   > DeltaRotation = RigVMFunctionReferenceNode(
    Rotation Order=ZYX
    child=head
    debugRotation=false
    max=40
    parent=head_null
    parentType=Null
    premult=1
    primAxis=2
    twistAxis=(X=1,Y=0,Z=0)
   )
   > &RigUnit_ModifyTransforms(
    ItemToModify="((Item=(Type=Bone,Name=\"base_hair_front_side_02_01_lf\"),Transform=(Rotation=(X=0,Y=0.2588,Z=0,W=0.9659),Translation=(X=0.0594,Y=0,Z=-0.2218),Scale3D=(X=1,Y=1,Z=1))),(Item=(Type=Bone,Name=\"base_hair_front_side_01_01_lf\"),Transform=(Rotation=(X=0,Y=-0.0872,Z ...(truncated)"
    WeightMinimum=0
    Weight=@DeltaRotation.outputWeight
   )
   n2 = &RigUnit_SphericalPoseReader(
    ActiveRegionScaleFactors=(PositiveWidth=1,NegativeWidth=1,PositiveHeight=1.2,NegativeHeight=1.2)
    ActiveRegionSize=0.1
    Debug=(bDrawDebug=false,bDraw2D=False,bDrawLocalAxes=False,DebugScale=25,DebugSegments=20,DebugThickness=0.25)
    DriverAxis=(X=1,Y=0,Z=0)
    DriverItem=(Type=Bone,Name="spine_05")
    ExecutePin=()
    FalloffRegionScaleFactors=(PositiveWidth=1,NegativeWidth=1,PositiveHeight=1,NegativeHeight=1)
    FalloffSize=0.5
    FlipHeightScaling=False
    FlipWidthScaling=False
    OptionalParentItem=(Type=Bone,Name="head")
    RotationOffset=(X=-90,Y=0,Z=180)
   )
   > &RigUnit_ModifyTransforms(
    ItemToModify=((Item=(Type=Bone,Name="dyn_hair_master"),Transform=(Rotation=(X=0,Y=-0.4226,Z=0,W=0.9063),Translation=(X=0,Y=0,Z=0),Scale3D=(X=1,Y=1,Z=1))))
    WeightMinimum=0
    Weight=@n2.OutputParam
   )
   n3 = &RigUnit_SphericalPoseReader(
    ActiveRegionScaleFactors=(PositiveWidth=1,NegativeWidth=1,PositiveHeight=1.2,NegativeHeight=1.2)
    ActiveRegionSize=0.1
    Debug=(bDrawDebug=true,bDraw2D=False,bDrawLocalAxes=False,DebugScale=25,DebugSegments=20,DebugThickness=0.25)
    DriverAxis=(X=1,Y=1,Z=0)
    DriverItem=(Type=Bone,Name="spine_05")
    ExecutePin=()
    FalloffRegionScaleFactors=(PositiveWidth=1,NegativeWidth=1,PositiveHeight=1,NegativeHeight=1)
    FalloffSize=0.5
    FlipHeightScaling=False
    FlipWidthScaling=False
    OptionalParentItem=(Type=Bone,Name="head")
    RotationOffset=(X=0,Y=0,Z=0)
   )
   > &RigUnit_ModifyTransforms(
    ItemToModify="((Item=(Type=Bone,Name=\"dyn_hair_master\"),Transform=(Rotation=(X=0,Y=-0.1736,Z=0,W=0.9848),Translation=(X=0,Y=0,Z=0),Scale3D=(X=1,Y=1,Z=1))),(Item=(Type=Bone,Name=\"base_hair_r_0\"),Transform=(Rotation=(X=0,Y=-0.1736,Z=0,W=0.9848),Translation=(X=0,Y=0,Z=0),S ...(truncated)"
    WeightMinimum=0
    Weight=@n3.OutputParam
   )
   > n4 = &RigUnit_SphericalPoseReader(
    ActiveRegionScaleFactors=(PositiveWidth=1,NegativeWidth=1,PositiveHeight=1.2,NegativeHeight=1.2)
    ActiveRegionSize=0.1
    Debug=(bDrawDebug=true,bDraw2D=False,bDrawLocalAxes=False,DebugScale=25,DebugSegments=20,DebugThickness=0.25)
    DriverAxis=(X=1,Y=1,Z=0)
    DriverItem=(Type=Bone,Name="spine_05")
    FalloffRegionScaleFactors=(PositiveWidth=1,NegativeWidth=1,PositiveHeight=1,NegativeHeight=1)
    FalloffSize=0.5
    FlipHeightScaling=False
    FlipWidthScaling=False
    OptionalParentItem=(Type=Bone,Name="head")
    RotationOffset=(X=180,Y=0,Z=0)
   )
   > &RigUnit_ModifyTransforms(
    ItemToModify=((Item=(Type=Bone,Name="dyn_hair_master"),Transform=(Rotation=(X=0,Y=-0.1736,Z=0,W=0.9848),Translation=(X=0,Y=0,Z=0),Scale3D=(X=1,Y=1,Z=1))))
    WeightMinimum=0
    Weight=@n4.OutputParam
   )
  ]}
]
hierarchy [
 root {loc=<0 0 0>, rot=<0 0 0 1>}
 pelvis {parent=root, loc=<0 -0.5301 98.0571>, rot=<0.0127 0.707 -0.0127 -0.707>}
 spine_01 {parent=pelvis, loc=<6.1208 -0.9328 0>, rot=<0 0 0.034 -0.9994>}
 spine_02 {parent=spine_01, loc=<6.5997 -0.2095 0>, rot=<0 0 0.0238 -0.9997>}
 spine_03 {parent=spine_02, loc=<6.4914 -0.1237 0>, rot=<0 0 0.0149 -0.9999>}
 spine_04 {parent=spine_03, loc=<6.9598 0.0794 0>, rot=<0 0 -0.0132 -0.9999>}
 spine_05 {parent=spine_04, loc=<6.9987 0 0>, rot=<0 0 -0.0015 -1>}
 neck_01 {parent=spine_05, loc=<11.3138 0.1529 0>, rot=<0 0 -0.2003 -0.9797>}
 neck_02 {parent=neck_01, loc=<4.5627 0 0>, rot=<0 0 0.0504 -0.9987>}
 head {parent=neck_02, loc=<5.3914 0 0>, rot=<0 0 0.115 -0.9934>}
 base_hair_master {parent=head, loc=<10.7835 -4.2 -0.8951>, rot=<0.0645 -0.7042 0.7042 -0.0645>}
 dyn_hair_master {parent=base_hair_master, loc=<0 0 0>, rot=<0 0 0 1>}
 base_hair_front_side_01_01_rt {parent=head, loc=<8.6874 5.7361 -6.5591>, rot=<0.0512 -0.7069 -0.7055 -0.0009>}
 dyn_hair_front_side_01_01_rt {parent=base_hair_front_side_01_01_rt, loc=<0 0 0>, rot=<0 0 0 1>}
 dyn_hair_front_side_01_02_rt {parent=dyn_hair_front_side_01_01_rt, loc=<4.9471 0 0>, rot=<0 0 -0.1983 -0.9801>}
 dyn_hair_front_side_01_03_rt {parent=dyn_hair_front_side_01_02_rt, loc=<3.0772 0 0>, rot=<0 0.2209 0 -0.9753>}
 base_hair_front_side_02_01_rt {parent=head, loc=<-1.6623 -2.8932 -6.7429>, rot=<-0.0029 0.7071 0.7071 -0.0029>}
 dyn_hair_front_side_02_01_rt {parent=base_hair_front_side_02_01_rt, loc=<0 0 0>, rot=<0 0 0 1>}
 dyn_hair_front_side_02_02_rt {parent=dyn_hair_front_side_02_01_rt, loc=<4.9471 0 0>, rot=<-0.247 0.2939 0.0789 -0.92>}
 dyn_hair_front_side_02_03_rt {parent=dyn_hair_front_side_02_02_rt, loc=<4.9471 0 0>, rot=<0 0.2554 0 -0.9668>}
 base_hair_front_side_01_01_lf {parent=head, loc=<9.5928 0.3401 7.0703>, rot=<-0.0341 0.7006 0.7052 -0.1035>}
 dyn_hair_front_side_01_01_lf {parent=base_hair_front_side_01_01_lf, loc=<0 0 0>, rot=<0 0 0 1>}
 dyn_hair_front_side_01_02_lf {parent=dyn_hair_front_side_01_01_lf, loc=<6.8687 0 0>, rot=<0 0.0934 0 -0.9956>}
 dyn_hair_front_side_01_03_lf {parent=dyn_hair_front_side_01_02_lf, loc=<5.6428 0 0>, rot=<0 0 0 1>}
 dyn_hair_front_side_01_04_lf {parent=dyn_hair_front_side_01_03_lf, loc=<4.0619 0 0>, rot=<0 0 0 1>}
 base_hair_front_side_02_01_lf {parent=head, loc=<8.155 5.2341 5.854>, rot=<-0.06 0.6949 0.7003 -0.1524>}
 dyn_hair_front_side_02_01_lf {parent=base_hair_front_side_02_01_lf, loc=<0 0 0>, rot=<0 0 0 1>}
 dyn_hair_front_side_02_02_lf {parent=dyn_hair_front_side_02_01_lf, loc=<6.8443 0 0>, rot=<0 0.0884 0 -0.9961>}
 dyn_hair_front_side_02_03_lf {parent=dyn_hair_front_side_02_02_lf, loc=<7.1791 0 0>, rot=<0 0.1114 0 -0.9938>}
 dyn_hair_front_side_02_04_lf {parent=dyn_hair_front_side_02_03_lf, loc=<4.4307 0 0>, rot=<0 0 0 1>}
 base_glasses {parent=head, loc=<7.2257 -2.3736 0>, rot=<0.4979 0.5021 0.5021 0.4979>}
 dyn_glasses {parent=base_glasses, loc=<0 0 0>, rot=<0 0 0 1>}
 base_hair_front_01 {parent=head, loc=<8.7757 9.5413 0.2385>, rot=<-0.0029 0.7071 0.7071 -0.0029>}
 dyn_hair_front_01 {parent=base_hair_front_01, loc=<0 0 0>, rot=<0 0 0 1>}
 base_hair_front_02 {parent=head, loc=<8.7855 8.351 3.9452>, rot=<-0.0029 0.7071 0.7071 -0.0029>}
 dyn_hair_front_02 {parent=base_hair_front_02, loc=<0 0 0>, rot=<0 0 0 1>}
 base_hair_rt {parent=head, loc=<10.0524 -4.2736 4.8749>, rot=<0.0758 -0.703 0.703 -0.0758>}
 dyn_hair_01_rt {parent=base_hair_rt, loc=<0 0 0>, rot=<0 0 0 1>}
 dyn_hair_02_rt {parent=dyn_hair_01_rt, loc=<11.5986 0 0>, rot=<0 -0.0332 0 -0.9995>}
 dyn_hair_03_rt {parent=dyn_hair_02_rt, loc=<6.9009 0 0>, rot=<0 0.2606 0 -0.9654>}
 dyn_hair_04_rt {parent=dyn_hair_03_rt, loc=<7.3661 0 0>, rot=<0 -0.3103 0 -0.9506>}
 base_hair_lf {parent=head, loc=<10.0524 -4.2736 -6.6651>, rot=<0.0758 -0.703 0.703 -0.0758>}
 dyn_hair_01_lf {parent=base_hair_lf, loc=<0 0 0>, rot=<0 0 0 1>}
 dyn_hair_02_lf {parent=dyn_hair_01_lf, loc=<11.5986 0 0>, rot=<0 -0.0332 0 -0.9995>}
 dyn_hair_03_lf {parent=dyn_hair_02_lf, loc=<6.9009 0 0>, rot=<0 0.2606 0 -0.9654>}
 dyn_hair_04_lf {parent=dyn_hair_03_lf, loc=<7.3661 0 0>, rot=<0 -0.3103 0 -0.9506>}
 base_hair_cr {parent=head, loc=<10.0524 -4.2736 -0.8951>, rot=<0.0758 -0.703 0.703 -0.0758>}
 dyn_hair_01_cr {parent=base_hair_cr, loc=<0 0 0>, rot=<0 0 0 1>}
 dyn_hair_02_cr {parent=dyn_hair_01_cr, loc=<11.5986 0 0>, rot=<0 -0.0332 0 -0.9995>}
 dyn_hair_03_cr {parent=dyn_hair_02_cr, loc=<6.9009 0 0>, rot=<0 0.2606 0 -0.9654>}
 dyn_hair_04_cr {parent=dyn_hair_03_cr, loc=<7.3661 0 0>, rot=<0 -0.3103 0 -0.9506>}
 head_null {parent=neck_01, type=Null, loc=<9.9266 -0.5431 0>, rot=<0 0 -0.1649 0.9863>}
 clavicle_l {parent=spine_05, loc=<5.4831 -0.1053 -3.2946>, rot=<0.0465 -0.7289 -0.009 -0.6829>}
 upperarm_l {parent=clavicle_l, loc=<9.8658 -1.2136 -0.4105>, rot=<0.0164 -0.3245 -0.0269 -0.9454>}
 clavicle_r {parent=spine_05, loc=<5.4831 -0.1053 3.2946>, rot=<0.7289 0.0465 -0.6829 0.009>}
 upperarm_r {parent=clavicle_r, loc=<-9.8658 1.2136 0.4105>, rot=<0.0164 -0.3245 -0.0269 -0.9454>}
 clavicle_r_null {parent=spine_05, type=Null, loc=<5.483 -0.1053 3.2946>, rot=<0.7289 0.0465 -0.6829 0.009>}
 base_head_aim_tgt {parent=spine_05, loc=<9.017 -11.714 0>, rot=<0.0616 -0.7044 -0.0616 -0.7044>}
 head_aim_tgt {parent=base_head_aim_tgt, loc=<0 0 0>, rot=<0 0 0 1>}
 Aim_Pistol_Crouch {type=Curve, loc=<0 0 0>, rot=<0 0 0 1>}
 Aim_Up {type=Curve, loc=<0 0 0>, rot=<0 0 0 1>}
 Aim_Up_Fusile {type=Curve, loc=<0 0 0>, rot=<0 0 0 1>}
]
variables [
 IsCrouch {type=bool, default=False}
 isTargeting {type=bool, default=False}
]
