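# Physics asset for the F_MED_AntiPesto female face accessory, setting up three dynamic hair chains (center, left, right).
# Each chain hangs from a base_hair bone jointed to neck_02 and runs through four capsule segments, with swing limits widening toward the tip and zero twist limit.
# Capsules on the neck, upper spine and clavicles form the body volume around the hair, linked by joints with free swing and twist.
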PhysicsAsset /BRCosmetics/Characters/Player/Female/Medium/Bodies/F_MED_AntiPesto/Meshes/Parts/F_MED_AntiPesto_FaceAcc_Physics
bodies {
 base_hair_cr [{shapeType=Capsule, length=0.505, radius=0.505, shapeName=base_hair_cr_capsule, rotation=<rot 0 -90 0>}]
 base_hair_lf [{shapeType=Capsule, length=0.505, radius=0.505, shapeName=base_hair_lf_capsule, rotation=<rot 0 -90 0>}]
 base_hair_rt [{shapeType=Capsule, length=0.505, radius=0.505, shapeName=base_hair_rt_capsule, rotation=<rot 0 -90 0>}]
 clavicle_l [{shapeType=Capsule, length=4.9228, radius=5.7886, shapeName=clavicle_l_capsule, center=<6.3703 -2.0871 0.0251>, rotation=<rot 90 -179.9999 0>}]
 clavicle_r [{
   shapeType=Capsule
   length=4.7873
   radius=5.9241
   shapeName=clavicle_r_capsule
   center=<-6.331 1.3506 0.0875>
   rotation=<rot 85 -0.0002 -0.0004>
  }]
 dyn_hair_01_cr [{shapeType=Capsule, length=14.1071, radius=2, shapeName=dyn_hair_01_rt_capsule, center=<15.2909 0 0.4897>, rotation=<rot 80.0001 -180 -180>}]
 dyn_hair_01_lf [{shapeType=Capsule, length=14.1071, radius=2, shapeName=dyn_hair_01_rt_capsule, center=<15.2909 0 0.4897>, rotation=<rot 80.0001 -180 -180>}]
 dyn_hair_01_rt [{shapeType=Capsule, length=14.1071, radius=2, shapeName=dyn_hair_01_rt_capsule, center=<15.2909 0 0.4897>, rotation=<rot 80.0001 -180 -180>}]
 dyn_hair_02_cr [{shapeType=Capsule, length=5.8123, radius=1.941, shapeName=dyn_hair_02_rt_capsule, center=<3.6175 0 2.1204>, rotation=<rot 65 -180 -180>}]
 dyn_hair_02_lf [{
   shapeType=Capsule
   length=4.5227
   radius=1.8541
   shapeName=dyn_hair_02_rt_capsule
   center=<5.0095 1 -0.3161>
   rotation=<rot 70.0002 -180 -180>
  }]
 dyn_hair_02_rt [{
   shapeType=Capsule
   length=2.5629
   radius=2.6429
   shapeName=dyn_hair_02_rt_capsule
   center=<4.4062 -2.2662 -0.5784>
   rotation=<rot 80.0001 -180 -180>
  }]
 dyn_hair_03_cr [{shapeType=Capsule, length=5.3259, radius=2.6741, shapeName=dyn_hair_03_cr_capsule, center=<4.8355 0 0.825>, rotation=<rot 79.9999 0 0>}]
 dyn_hair_03_lf [{
   shapeType=Capsule
   length=6.0702
   radius=1.9298
   shapeName=dyn_hair_03_lf_capsule
   center=<4.236 3.2156 0.1369>
   rotation=<rot 85.0001 -180 -180>
  }]
 dyn_hair_03_rt [{
   shapeType=Capsule
   length=5.6795
   radius=2.5311
   shapeName=dyn_hair_03_rt_capsule
   center=<3.7014 -3.7648 0.0006>
   rotation=<rot 85.0005 -180 -180>
  }]
 dyn_hair_04_cr [{shapeType=Capsule, length=8.0522, radius=2.8211, shapeName=dyn_hair_04_cr_capsule, center=<7.3466 0 0>, rotation=<rot 90 0 0>}]
 dyn_hair_04_lf [{shapeType=Capsule, length=8.1395, radius=2.7337, shapeName=dyn_hair_04_lf_capsule, center=<7.3466 0 0>, rotation=<rot 90 0 0>}]
 dyn_hair_04_rt [{shapeType=Capsule, length=8.2726, radius=2.6006, shapeName=dyn_hair_04_rt_capsule, center=<7.3466 -3.4272 0>, rotation=<rot 90 0 0>}]
 neck_02 [{shapeType=Capsule, length=18.4237, radius=12.3308, shapeName=neck_02_capsule, center=<3.5779 12.2787 0>, rotation=<rot 0 -90 0>}]
 spine_04 [
  {shapeType=Capsule, length=17.1975, radius=7.8025, shapeName=spine_05_capsule, center=<-9.7587 2.9014 0>, rotation=<rot 0 -90 0>}
  {shapeType=Capsule, length=19.0042, radius=7.207, shapeName=spine_04_capsule, center=<-0.068 -1.8225 0>}
 ]
 spine_05 [{shapeType=Capsule, length=17.0286, radius=8.4421, shapeName=spine_05_capsule, center=<2.2882 -0.6259 0>, rotation=<rot 0 -90 0>}]
}
constraints [
 {
  bone1Name=clavicle_l
  bone2Name=spine_05
  swing1LimitDegrees=45
  swing1Motion=Free
  swing2LimitDegrees=45
  swing2Motion=Free
  twistLimitDegrees=45
  twistMotion=Free
 }
 {
  bone1Name=clavicle_r
  bone2Name=spine_05
  swing1LimitDegrees=45
  swing1Motion=Free
  swing2LimitDegrees=45
  swing2Motion=Free
  twistLimitDegrees=45
  twistMotion=Free
 }
 {
  bone1Name=spine_05
  bone2Name=spine_04
  swing1LimitDegrees=45
  swing1Motion=Free
  swing2LimitDegrees=45
  swing2Motion=Free
  twistLimitDegrees=45
  twistMotion=Free
 }
 {
  bone1Name=base_hair_cr
  bone2Name=neck_02
  swing1LimitDegrees=45
  swing1Motion=Free
  swing2LimitDegrees=45
  swing2Motion=Free
  twistLimitDegrees=45
  twistMotion=Free
 }
 {
  bone1Name=base_hair_lf
  bone2Name=neck_02
  swing1LimitDegrees=45
  swing1Motion=Free
  swing2LimitDegrees=45
  swing2Motion=Free
  twistLimitDegrees=45
  twistMotion=Free
 }
 {
  bone1Name=base_hair_rt
  bone2Name=neck_02
  swing1LimitDegrees=45
  swing1Motion=Free
  swing2LimitDegrees=45
  swing2Motion=Free
  twistLimitDegrees=45
  twistMotion=Free
 }
 {bone1Name=dyn_hair_01_rt, bone2Name=base_hair_rt, swing1LimitDegrees=0, swing2LimitDegrees=5, twistLimitDegrees=0}
 {bone1Name=dyn_hair_02_rt, bone2Name=dyn_hair_01_rt, swing1LimitDegrees=0, swing2LimitDegrees=20, twistLimitDegrees=0}
 {bone1Name=dyn_hair_03_rt, bone2Name=dyn_hair_02_rt, swing1LimitDegrees=5, swing2LimitDegrees=40, twistLimitDegrees=0}
 {bone1Name=dyn_hair_04_rt, bone2Name=dyn_hair_03_rt, swing1LimitDegrees=10, swing2LimitDegrees=40, twistLimitDegrees=0}
 {bone1Name=dyn_hair_01_lf, bone2Name=base_hair_lf, swing1LimitDegrees=0, swing2LimitDegrees=5, twistLimitDegrees=0}
 {bone1Name=dyn_hair_02_lf, bone2Name=dyn_hair_01_lf, swing1LimitDegrees=0, swing2LimitDegrees=20, twistLimitDegrees=0}
 {bone1Name=dyn_hair_03_lf, bone2Name=dyn_hair_02_lf, swing1LimitDegrees=5, swing2LimitDegrees=40, twistLimitDegrees=0}
 {bone1Name=dyn_hair_04_lf, bone2Name=dyn_hair_03_lf, swing1LimitDegrees=10, swing2LimitDegrees=40, twistLimitDegrees=0}
 {bone1Name=dyn_hair_01_cr, bone2Name=base_hair_cr, swing1LimitDegrees=0, swing2LimitDegrees=5, twistLimitDegrees=0}
 {bone1Name=dyn_hair_02_cr, bone2Name=dyn_hair_01_cr, swing1LimitDegrees=0, swing2LimitDegrees=20, twistLimitDegrees=0}
 {bone1Name=dyn_hair_03_cr, bone2Name=dyn_hair_02_cr, swing1LimitDegrees=5, swing2LimitDegrees=40, twistLimitDegrees=0}
 {bone1Name=dyn_hair_04_cr, bone2Name=dyn_hair_03_cr, swing1LimitDegrees=10, swing2LimitDegrees=40, twistLimitDegrees=0}
 {
  bone1Name=neck_02
  bone2Name=spine_05
  swing1LimitDegrees=45
  swing1Motion=Free
  swing2LimitDegrees=45
  swing2Motion=Free
  twistLimitDegrees=45
  twistMotion=Free
 }
]
