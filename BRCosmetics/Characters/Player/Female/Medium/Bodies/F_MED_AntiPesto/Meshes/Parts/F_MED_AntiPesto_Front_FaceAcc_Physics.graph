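# Physics asset for the AntiPesto medium female character's front face accessory, covering dynamic front-side hair strands.
# Three capsule chains, two left and one right, hang from small base bodies linked to neck_01.
# Capsules on upper spine, neck, clavicles and only the right upper arm form the body volume, joined by free constraints.
# Strand joints mostly have tight swing limits and locked or near-zero twist, but the right strand's first dynamic joint is free.
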PhysicsAsset /BRCosmetics/Characters/Player/Female/Medium/Bodies/F_MED_AntiPesto/Meshes/Parts/F_MED_AntiPesto_Front_FaceAcc_Physics
bodies {
 base_hair_front_side_01_01_lf [{shapeType=Capsule, length=0.505, radius=0.505, shapeName=base_hair_front_side_01_01_lf_capsule, rotation=<rot 0 -90 0>}]
 base_hair_front_side_02_01_lf [{shapeType=Capsule, length=0.505, radius=0.505, shapeName=base_hair_front_side_02_01_lf_capsule, rotation=<rot 0 -90 0>}]
 base_hair_front_side_02_01_rt [{shapeType=Capsule, length=0.505, radius=0.505, shapeName=base_hair_front_side_02_01_rt_capsule, rotation=<rot 0 -90 0>}]
 clavicle_l [{shapeType=Capsule, length=10, radius=7, shapeName=clavicle_l_capsule, center=<3.343 -1.2682 -1.3222>, rotation=<rot 85 -0.0004 179.9995>}]
 clavicle_r [{
   shapeType=Capsule
   length=9.396
   radius=7.604
   shapeName=clavicle_r_capsule
   center=<-3.3005 0.8047 1.8085>
   rotation=<rot 85 -0.0007 -0.0007>
  }]
 dyn_hair_front_side_01_01_lf [{shapeType=Capsule, length=4, radius=1.5, shapeName=dyn_hair_front_side_01_01_lf_capsule, center=<5 0 0>, rotation=<rot 90 0 0>}]
 dyn_hair_front_side_01_02_lf [{
   shapeType=Capsule
   length=4.5889
   radius=1.5207
   shapeName=dyn_hair_front_side_01_02_lf_capsule
   center=<2.2132 -0.0488 0.0189>
   rotation=<rot 15.2823 -90.0001 94.9573>
  }]
 dyn_hair_front_side_01_03_lf [{
   shapeType=Capsule
   length=2.5543
   radius=2.1146
   shapeName=dyn_hair_front_side_01_03_lf_capsule
   center=<2.4679 0.0044 -0.7731>
   rotation=<rot -83.3031 -35.0598 54.3864>
  }]
 dyn_hair_front_side_01_04_lf [{
   shapeType=Capsule
   length=5.4023
   radius=1.3771
   shapeName=dyn_hair_front_side_01_04_lf_capsule
   center=<3.305 0.933 0.1089>
   rotation=<rot 0 103.6614 -38.9334>
  }]
 dyn_hair_front_side_02_01_lf [{shapeType=Capsule, length=4, radius=1.5, shapeName=dyn_hair_front_side_02_01_lf_capsule, center=<5 0 0>, rotation=<rot 90 0 0>}]
 dyn_hair_front_side_02_01_rt [{
   shapeType=Capsule
   length=1.2791
   radius=0.9582
   shapeName=dyn_hair_front_side_02_01_rt_capsule
   center=<4.2346 -0.3889 0.0021>
   rotation=<rot -17.2038 -90 2.3568>
  }]
 dyn_hair_front_side_02_02_lf [{
   shapeType=Capsule
   length=3.912
   radius=2.5085
   shapeName=dyn_hair_front_side_02_02_lf_capsule
   center=<2.6655 0.5516 0.457>
   rotation=<rot 0 92.8971 -89.7688>
  }]
 dyn_hair_front_side_02_02_rt [{
   shapeType=Capsule
   length=2.4682
   radius=1.2028
   shapeName=dyn_hair_front_side_02_02_rt_capsule
   center=<2.5427 0.001 -0.2142>
   rotation=<rot 0 88.997 -87.2773>
  }]
 dyn_hair_front_side_02_03_lf [{
   shapeType=Capsule
   length=4.3304
   radius=1.3729
   shapeName=dyn_hair_front_side_02_03_lf_capsule
   center=<2.6611 0.6632 -0.6718>
   rotation=<rot -6.9999 98.559 -87.3345>
  }]
 dyn_hair_front_side_02_03_rt [{
   shapeType=Capsule
   length=2.4721
   radius=1.1658
   shapeName=dyn_hair_front_side_02_03_rt_capsule
   center=<1.2823 -0.3277 0.4135>
   rotation=<rot 0.0001 81.0109 -95.7184>
  }]
 dyn_hair_front_side_02_04_lf [{
   shapeType=Capsule
   length=8.1906
   radius=1.1828
   shapeName=dyn_hair_front_side_02_04_lf_capsule
   center=<5.5521 0.3496 2.3497>
   rotation=<rot 0 63.3423 -54.6827>
  }]
 neck_01 [{shapeType=Capsule, length=7.717, radius=7.0966, shapeName=neck_01_capsule, center=<0 2.6644 0>, rotation=<rot 90 -180 0>}]
 spine_05 [{shapeType=Capsule, length=16.1466, radius=8.8534, shapeName=spine_05_capsule, center=<-3.4331 2.6799 0>, rotation=<rot 0 -90 0>}]
 upperarm_r [{shapeType=Capsule, length=18.3615, radius=5.7096, shapeName=upperarm_r_capsule, center=<-13.0317 0 0>, rotation=<rot 90 0 0>}]
}
constraints [
 {
  bone1Name=clavicle_l
  bone2Name=spine_05
  swing1LimitDegrees=45
  swing1Motion=Free
  swing2LimitDegrees=45
  swing2Motion=Free
  twistLimitDegrees=45
  twistMotion=Free
 }
 {
  bone1Name=clavicle_r
  bone2Name=spine_05
  swing1LimitDegrees=45
  swing1Motion=Free
  swing2LimitDegrees=45
  swing2Motion=Free
  twistLimitDegrees=45
  twistMotion=Free
 }
 {
  bone1Name=upperarm_r
  bone2Name=clavicle_r
  swing1LimitDegrees=45
  swing1Motion=Free
  swing2LimitDegrees=45
  swing2Motion=Free
  twistLimitDegrees=45
  twistMotion=Free
 }
 {
  bone1Name=base_hair_front_side_01_01_lf
  bone2Name=neck_01
  swing1LimitDegrees=45
  swing1Motion=Free
  swing2LimitDegrees=45
  swing2Motion=Free
  twistLimitDegrees=45
  twistMotion=Free
 }
 {
  bone1Name=dyn_hair_front_side_01_01_lf
  bone2Name=base_hair_front_side_01_01_lf
  swing1LimitDegrees=10
  swing2LimitDegrees=7
  twistLimitDegrees=0
 }
 {
  bone1Name=dyn_hair_front_side_01_02_lf
  bone2Name=dyn_hair_front_side_01_01_lf
  swing1LimitDegrees=5
  swing2LimitDegrees=10
  twistLimitDegrees=45
  twistMotion=Locked
 }
 {
  bone1Name=dyn_hair_front_side_01_03_lf
  bone2Name=dyn_hair_front_side_01_02_lf
  swing1LimitDegrees=5
  swing2LimitDegrees=12
  twistLimitDegrees=10
 }
 {
  bone1Name=base_hair_front_side_02_01_lf
  bone2Name=neck_01
  swing1LimitDegrees=45
  swing1Motion=Free
  swing2LimitDegrees=45
  swing2Motion=Free
  twistLimitDegrees=45
  twistMotion=Free
 }
 {
  bone1Name=dyn_hair_front_side_02_01_lf
  bone2Name=base_hair_front_side_02_01_lf
  swing1LimitDegrees=10
  swing2LimitDegrees=7
  twistLimitDegrees=0
 }
 {
  bone1Name=dyn_hair_front_side_02_02_lf
  bone2Name=dyn_hair_front_side_02_01_lf
  swing1LimitDegrees=5
  swing2LimitDegrees=10
  twistLimitDegrees=45
  twistMotion=Locked
 }
 {
  bone1Name=dyn_hair_front_side_02_03_lf
  bone2Name=dyn_hair_front_side_02_02_lf
  swing1LimitDegrees=20
  swing2LimitDegrees=10
  twistLimitDegrees=45
  twistMotion=Locked
 }
 {
  bone1Name=base_hair_front_side_02_01_rt
  bone2Name=neck_01
  swing1LimitDegrees=45
  swing1Motion=Free
  swing2LimitDegrees=45
  swing2Motion=Free
  twistLimitDegrees=45
  twistMotion=Free
 }
 {
  bone1Name=dyn_hair_front_side_02_01_rt
  bone2Name=base_hair_front_side_02_01_rt
  swing1LimitDegrees=45
  swing1Motion=Free
  swing2LimitDegrees=45
  swing2Motion=Free
  twistLimitDegrees=45
  twistMotion=Free
 }
 {
  bone1Name=dyn_hair_front_side_02_02_rt
  bone2Name=dyn_hair_front_side_02_01_rt
  swing1LimitDegrees=0
  swing2LimitDegrees=15
  twistLimitDegrees=45
  twistMotion=Locked
 }
 {
  bone1Name=dyn_hair_front_side_02_03_rt
  bone2Name=dyn_hair_front_side_02_02_rt
  swing1LimitDegrees=10
  swing2LimitDegrees=15
  twistLimitDegrees=45
  twistMotion=Locked
 }
 {
  bone1Name=neck_01
  bone2Name=spine_05
  swing1LimitDegrees=45
  swing1Motion=Free
  swing2LimitDegrees=45
  swing2Motion=Free
  twistLimitDegrees=45
  twistMotion=Free
 }
 {
  bone1Name=dyn_hair_front_side_01_04_lf
  bone2Name=dyn_hair_front_side_01_03_lf
  swing1LimitDegrees=20
  swing2LimitDegrees=20
  twistLimitDegrees=0
 }
 {
  bone1Name=dyn_hair_front_side_02_04_lf
  bone2Name=dyn_hair_front_side_02_03_lf
  swing1LimitDegrees=20
  swing2LimitDegrees=20
  twistLimitDegrees=0
 }
]
