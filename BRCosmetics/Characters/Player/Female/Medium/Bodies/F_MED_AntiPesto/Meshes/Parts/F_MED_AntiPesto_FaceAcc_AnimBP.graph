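# Animation Blueprint for the AntiPesto female face accessory mesh, driving hair secondary motion.
# Parent and body-mesh arm poses feed a crouch/targeting Control Rig, Anim Dynamics front-hair springs, then fixed base-hair offsets.
# Head-space RigidBody simulation switches to a backpack physics asset for back bling or capes, followed by a front-hair pass and aim-curve-driven bone corrections.
# The EventGraph detects back bling and computes a gravity override the AnimGraph never reads.
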AnimBlueprint /BRCosmetics/Characters/Player/Female/Medium/Bodies/F_MED_AntiPesto/Meshes/Parts/F_MED_AntiPesto_FaceAcc_AnimBP
target_skeleton=&/F_MED_AntiPesto_FaceAcc_Skeleton
graphs [
 AnimGraph {nodes [
   n1 = "Transform (Modify) Bone - Bone: dyn_hair_02_rt"!(
    ComponentPose=Local To Component(LocalPose=Use cached pose 'ModifyFixedPose'())
    AlphaCurveName:Name=Aim_Up_Fusile
    Translation="0,0,0"
    Rotation="-20,0,0"
    Scale="1,1,1"
   ) -> (Pose) {
    alphaInputType=Curve
    rotationMode=BMM_Additive
    rotationSpace=BCS_BoneSpace
    translationMode=BMM_Additive
    translationSpace=BCS_BoneSpace
    alphaBoolBlend {blendOption=Linear}
    boneToModify {boneName=dyn_hair_02_rt}
    rotation {pitch=-20}
   }
   n2 = "Transform (Modify) Bone - Bone: dyn_hair_03_rt"!(ComponentPose=@n1, AlphaCurveName:Name=Aim_Up_Fusile, Translation="0,0,0", Rotation="-10,0,0", Scale="1,1,1") -> (Pose) {
    alphaInputType=Curve
    rotationMode=BMM_Additive
    rotationSpace=BCS_BoneSpace
    translationMode=BMM_Additive
    translationSpace=BCS_BoneSpace
    alphaBoolBlend {blendOption=Linear}
    boneToModify {boneName=dyn_hair_03_rt}
    rotation {pitch=-10}
   }
   n3 = "Transform (Modify) Bone - Bone: dyn_hair_front_side_02_03_lf"!(ComponentPose=@n2, AlphaCurveName:Name=Aim_Up, Translation="0,0,0", Rotation="-30,0,0", Scale="1,1,1") -> (Pose) {
    alphaInputType=Curve
    rotationMode=BMM_Additive
    rotationSpace=BCS_BoneSpace
    translationMode=BMM_Additive
    translationSpace=BCS_BoneSpace
    alphaBoolBlend {blendOption=Linear}
    boneToModify {boneName=dyn_hair_front_side_02_03_lf}
    rotation {pitch=-30}
   }
   n4 = "Transform (Modify) Bone - Bone: dyn_hair_front_side_01_03_lf"!(ComponentPose=@n3, AlphaCurveName:Name=Aim_Up, Translation="0,0,0", Rotation="-20,0,0", Scale="1,1,1") -> (Pose) {
    alphaInputType=Curve
    rotationMode=BMM_Additive
    rotationSpace=BCS_BoneSpace
    translationMode=BMM_Additive
    translationSpace=BCS_BoneSpace
    alphaBoolBlend {blendOption=Linear}
    boneToModify {boneName=dyn_hair_front_side_01_03_lf}
    rotation {pitch=-20}
   }
   n5 = "Transform (Modify) Bone - Bone: dyn_hair_front_side_01_03_lf"(ComponentPose=@n4, AlphaCurveName=Aim_Pistol_Crouch, Translation="0,0,0", Rotation="-10.908,-40,0", Scale="1,1,1") {
    alphaInputType=Curve
    rotationMode=BMM_Additive
    rotationSpace=BCS_BoneSpace
    translationMode=BMM_Additive
    translationSpace=BCS_BoneSpace
    alphaBoolBlend {blendOption=Linear}
    boneToModify {boneName=dyn_hair_front_side_01_03_lf}
    rotation {pitch=-10.908, yaw=-40}
   }
   n6 = "Transform (Modify) Bone - Bone: dyn_hair_front_side_02_03_lf"(ComponentPose=@n5, AlphaCurveName=Aim_Pistol_Crouch, Translation="0.922,2.009,-0.807", Rotation="-13.861,-20,5", Scale="1,1,1") {
    alphaInputType=Curve
    rotationMode=BMM_Additive
    rotationSpace=BCS_BoneSpace
    translationMode=BMM_Additive
    translationSpace=BCS_BoneSpace
    alphaBoolBlend {blendOption=Linear}
    boneToModify {boneName=dyn_hair_front_side_02_03_lf}
    rotation=<rot -13.861 -20 5>
    translation=<0.922 2.009 -0.807>
   }
   Output Pose(Result=Component To Local(ComponentPose=@n6)) {layerGroup=None}
   n7 = Control Rig!(Source=Use cached pose 'StartPose'(), Alpha, IsCrouch:bool=Get bIsCrouching(), isTargeting:bool=Get bIsTargeting()) -> (Pose) {
    alphaBoolBlend {blendOption=Linear}
    controlRigAssetReference {blueprintRigClass=&/F_MED_AntiPesto_FaceAcc_CtrlRig.F_MED_AntiPesto_FaceAcc_CtrlRig_C}
   }
   Save cached pose 'AnimDynPose'(Pose=@n7) {cacheName=AnimDynPose}
   n8 = "Transform (Modify) Bone - Bone: base_hair_lf"!(
    ComponentPose=Local To Component(LocalPose=Use cached pose 'FizPrePhysicsPose'())
    bAlphaBoolEnabled:bool=True
    Translation="0,0,0"
    Rotation="0,-15,0"
    Scale="1,1,1"
   ) -> (Pose) {
    alphaInputType=Bool
    rotationMode=BMM_Additive
    rotationSpace=BCS_BoneSpace
    alphaBoolBlend {blendInTime=0.3, blendOption=Linear, blendOutTime=0.3}
    boneToModify {boneName=base_hair_lf}
    rotation {yaw=-15}
   }
   n9 = "Transform (Modify) Bone - Bone: base_hair_rt"!(ComponentPose=@n8, bAlphaBoolEnabled:bool=True, Translation="0,0,0", Rotation="0,10,0", Scale="1,1,1") -> (Pose) {
    alphaInputType=Bool
    rotationMode=BMM_Additive
    rotationSpace=BCS_BoneSpace
    alphaBoolBlend {blendInTime=0.3, blendOption=Linear, blendOutTime=0.3}
    boneToModify {boneName=base_hair_rt}
    rotation {yaw=10}
   }
   n10 = "Transform (Modify) Bone - Bone: base_hair_lf"(ComponentPose=@n9, bAlphaBoolEnabled=True, Translation="0,0,0", Rotation="0,0,0", Scale="1,1,1") {
    alphaInputType=Bool
    rotationMode=BMM_Additive
    rotationSpace=BCS_BoneSpace
    alphaBoolBlend {blendInTime=0.3, blendOption=Linear, blendOutTime=0.3}
    boneToModify {boneName=base_hair_lf}
   }
   n11 = "Transform (Modify) Bone - Bone: base_hair_rt"(ComponentPose=@n10, bAlphaBoolEnabled=True, Translation="0,0,0", Rotation="0,0,0", Scale="1,1,1") {
    alphaInputType=Bool
    rotationMode=BMM_Additive
    rotationSpace=BCS_BoneSpace
    alphaBoolBlend {blendInTime=0.3, blendOption=Linear, blendOutTime=0.3}
    boneToModify {boneName=base_hair_rt}
   }
   n12 = "Transform (Modify) Bone - Bone: base_hair_rt"(ComponentPose=@n11, bAlphaBoolEnabled=True, Translation="0,0,3", Rotation="0,0,0.017", Scale="1,1,1") {
    alphaInputType=Bool
    rotationMode=BMM_Additive
    rotationSpace=BCS_BoneSpace
    translationMode=BMM_Additive
    translationSpace=BCS_BoneSpace
    alphaBoolBlend {blendInTime=0.3, blendOption=Linear, blendOutTime=0.3}
    boneToModify {boneName=base_hair_rt}
    rotation {roll=0.017}
    translation {z=3}
   }
   Save cached pose 'RigidBodyPose'(Pose=Component To Local(ComponentPose=@n12)) {cacheName=RigidBodyPose}
   n13 = RigidBody!(ComponentPose=Local To Component(LocalPose=Use cached pose 'RigidBodyPose'()), Alpha, OverrideWorldGravity:Vector="0,0,-980", ExternalForce) -> (Pose) {
    bClampLinearTranslationLimitToRefPose=true
    bOverrideWorldGravity=true
    overridePhysicsAsset=&/F_MED_AntiPesto_FaceAcc_Physics
    simulationSpace=BaseBoneSpace
    alphaBoolBlend {blendOption=Linear}
    baseBoneRef {boneName=head}
    overrideWorldGravity {z=-980}
    simSpaceSettings {worldAlpha=0.4, externalAngularVelocity {x=-0.5}, externalLinearDragV {x=-0.5}, externalLinearVelocity {x=-0.5}}
   }
   n14 = Component To Local(ComponentPose=@n13)
   n15 = RigidBody(ComponentPose=Local To Component(LocalPose=Use cached pose 'RigidBodyPose'()), OverrideWorldGravity="0,0,-980") {
    bClampLinearTranslationLimitToRefPose=true
    bOverrideWorldGravity=true
    overridePhysicsAsset=&/F_MED_AntiPesto_FaceAcc_Backpack_Physics
    simulationSpace=BaseBoneSpace
    alphaBoolBlend {blendOption=Linear}
    baseBoneRef {boneName=head}
    overrideWorldGravity {z=-980}
   }
   n16 = Component To Local(ComponentPose=@n15)
   n17 = Get is Wearing cape and backpack() -> (is Wearing cape and backpack:bool)
   n18 = RigidBody(
    ComponentPose=Local To Component(LocalPose=Blend Poses by bool(BlendPose_0=@n16, BlendPose_1=@n14, bActiveValue=@n17) {blendType=Linear})
    OverrideWorldGravity="0,0,-980"
   ) {
    bOverrideWorldGravity=true
    overridePhysicsAsset=&/F_MED_AntiPesto_Front_FaceAcc_Physics
    simulationSpace=BaseBoneSpace
    alphaBoolBlend {blendOption=Linear}
    baseBoneRef {boneName=head}
    overrideWorldGravity {z=-980}
    simSpaceSettings {worldAlpha=0.45}
   }
   Save cached pose 'ModifyFixedPose'(Pose=Component To Local(ComponentPose=@n18)) {cacheName=ModifyFixedPose}
   n19 = Anim Dynamics - Bone: dyn_hair_front_01(ComponentPose=Local To Component(LocalPose=Use cached pose 'AnimDynPose'()), Alpha=1, ExternalForce="0,0,0") {
    angularDampingOverride=1
    angularSpringConstant=20000
    bAngularSpring=true
    bOverrideAngularDamping=true
    bUseGravityOverride=true
    gravityScale=0.35
    simulationSpace=RootRelative
    alphaBoolBlend {blendOption=Linear}
    boundBone {boneName=dyn_hair_front_01}
    componentAppliedLinearAccClamp=<6000 6000 6000>
    componentLinearAccScale=<1 1 1>
    componentLinearVelScale=<1 1 1>
    gravityOverride {z=-980}
    physicsBodyDefinitions [{
      boundBone {boneName=dyn_hair_front_01}
      boxExtents=<20 30 30>
      constraintSetup {angularLimitsMax=<0 0 10>, angularLimitsMin=<0 -10 -10>, angularTarget=<1 0 0>}
      localJointOffset=<2 0 0>
     }]
   }
   n20 = Anim Dynamics - Bone: dyn_hair_front_02(ComponentPose=@n19, Alpha=1, ExternalForce="0,0,0") {
    angularDampingOverride=1
    angularSpringConstant=20000
    bAngularSpring=true
    bOverrideAngularDamping=true
    bUseGravityOverride=true
    gravityScale=0.35
    simulationSpace=RootRelative
    alphaBoolBlend {blendOption=Linear}
    boundBone {boneName=dyn_hair_front_02}
    componentAppliedLinearAccClamp=<6000 6000 6000>
    componentLinearAccScale=<1 1 1>
    componentLinearVelScale=<1 1 1>
    gravityOverride {z=-980}
    physicsBodyDefinitions [{
      boundBone {boneName=dyn_hair_front_02}
      boxExtents=<20 30 30>
      constraintSetup {angularLimitsMax=<0 0 10>, angularLimitsMin=<0 -10 -10>, angularTarget=<1 0 0>}
      localJointOffset=<2 0 0>
     }]
   }
   n21 = Anim Dynamics - Bone: dyn_hair_front_side_01_01_rt(ComponentPose=@n20, Alpha=1, ExternalForce="0,0,0") {
    angularDampingOverride=1
    angularSpringConstant=20000
    bAngularSpring=true
    bOverrideAngularDamping=true
    bUseGravityOverride=true
    gravityScale=0.35
    simulationSpace=RootRelative
    alphaBoolBlend {blendOption=Linear}
    boundBone {boneName=dyn_hair_front_side_01_01_rt}
    componentAppliedLinearAccClamp=<6000 6000 6000>
    componentLinearAccScale=<1 1 1>
    componentLinearVelScale=<1 1 1>
    gravityOverride {z=-980}
    physicsBodyDefinitions [{
      boundBone {boneName=dyn_hair_front_side_01_01_rt}
      boxExtents=<20 30 30>
      constraintSetup {angularLimitsMax=<0 0 0>, angularLimitsMin=<0 -15 -15>, angularTarget=<1 0 0>}
      localJointOffset=<4 0 0>
     }]
   }
   n22 = Anim Dynamics - Bone: dyn_hair_front_side_01_02_rt(ComponentPose=@n21, Alpha=1, ExternalForce="0,0,0") {
    angularDampingOverride=1
    angularSpringConstant=20000
    bAngularSpring=true
    bOverrideAngularDamping=true
    bUseGravityOverride=true
    gravityScale=0.35
    simulationSpace=RootRelative
    alphaBoolBlend {blendOption=Linear}
    boundBone {boneName=dyn_hair_front_side_01_02_rt}
    componentAppliedLinearAccClamp=<6000 6000 6000>
    componentLinearAccScale=<1 1 1>
    componentLinearVelScale=<1 1 1>
    gravityOverride {z=-980}
    physicsBodyDefinitions [{
      boundBone {boneName=dyn_hair_front_side_01_02_rt}
      boxExtents=<20 30 30>
      constraintSetup {angularLimitsMax=<0 0 0>, angularLimitsMin=<0 -15 -15>, angularTarget=<1 0 0>}
      localJointOffset=<4 0 0>
     }]
   }
   n23 = Anim Dynamics - Bone: dyn_hair_front_side_01_03_rt(ComponentPose=@n22, Alpha=1, ExternalForce="0,0,0") {
    angularDampingOverride=1
    angularSpringConstant=20000
    bAngularSpring=true
    bOverrideAngularDamping=true
    bUseGravityOverride=true
    gravityScale=0.35
    simulationSpace=RootRelative
    alphaBoolBlend {blendOption=Linear}
    boundBone {boneName=dyn_hair_front_side_01_03_rt}
    componentAppliedLinearAccClamp=<6000 6000 6000>
    componentLinearAccScale=<1 1 1>
    componentLinearVelScale=<1 1 1>
    gravityOverride {z=-980}
    physicsBodyDefinitions [{
      boundBone {boneName=dyn_hair_front_side_01_03_rt}
      boxExtents=<20 30 30>
      constraintSetup {angularLimitsMax=<0 0 0>, angularLimitsMin=<0 -15 -15>, angularTarget=<1 0 0>}
      localJointOffset=<4 0 0>
     }]
   }
   Save cached pose 'FizPrePhysicsPose'(Pose=Component To Local(ComponentPose=@n23)) {cacheName=FizPrePhysicsPose}
   n24 = Layered blend per bone(
    BasePose=Copy Pose From Mesh() {bCopyCurves=true, bUseAttachedParent=true}
    BlendPoses_0=Copy Pose From Mesh(SourceMeshComponent=Get Body_Mesh())
   ) {layerSetup [{branchFilters [{boneName=clavicle_l}, {boneName=clavicle_r}, {boneName=upperarm_r}, {boneName=upperarm_l}]}]}
   Save cached pose 'StartPose'(Pose=@n24) {cacheName=StartPose}
   Get isWearingBackbling()
  ]}
 EventGraph {nodes [
   Event OnCharacterCustomizationLoaded()
   > n1 = Cast To PlayerPawn_Athena(Object=TryGetPawnOwner())
   > Set Body_Mesh(Body_Mesh=GetSkeletalMeshForPartType(self=@n1))
   n2 = CheckForBackbling()
   > n3 = Set isWearingBackbling(isWearingBackbling=AND Boolean!(A=@n2, B=NOT Boolean(A=@n2.bIsBackblingACape), C:bool=NOT Boolean(A=@n2.bIsBackblingLong)) -> (ReturnValue))
   > n4 = Set Is Wearing Cape(Is Wearing Cape=OR Boolean(A=@n2.bIsBackblingACape, B=@n2.bIsBackblingLong))
   > Set is Wearing cape and backpack(is Wearing cape and backpack:bool=OR Boolean(A=@n3, B=@n4)) -> (Output_Get:bool)
   Event BlueprintUpdateAnimation()
   > n5 = Sequence()
   n6 = Get bIsSkyDiving()
   Set IsFallingJump(execute=@n5.then_1, IsFallingJump:bool=AND Boolean!(A=Get bIsFalling(), B=NOT Boolean(A=@n6)) -> (ReturnValue)) -> (Output_Get:bool)
   n7 = Gravity Override State Function DEPRECATED(
    execute=@n5.then_0
    CharacterPartAnimInstance=Self-Reference()
    GameTimeInSeconds=GetGameTimeInSeconds()
    GravityOverrideParams=Get Gravity Override Params()
   )
   > Set GravityOverride(GravityOverride=@n7)
  ]}
]
variables [
 Body_Mesh {type=SkeletalMeshComponent}
 get is ads up crouch rotation in place {type=boolean}
 Gravity Override Params {type=GravityOverrideParamsStruct, default {skydive_GoingUp {gravityOverride=<0 0 0>}}}
 GravityOverride {type=Vector}
 Is Wearing Cape {type=boolean}
 is Wearing cape and backpack {type=boolean}
 IsFallingJump {type=boolean}
 isWearingBackbling {type=boolean}
]
